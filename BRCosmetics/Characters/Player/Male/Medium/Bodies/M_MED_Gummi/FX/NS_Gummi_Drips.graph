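# Niagara system for the Gummi character body that sheds occasional sprite drips from the skeletal mesh.
# Each drip gets one of eight spawn groups, which indexes filtered arm, hand and calf bones for its spawn location.
# Curves ramp up downward gravity while stretching the sprite vertically and narrowing it.
# Drips_Low is a GPU copy of Drips; the collision modules in Drips and the Empty burst-velocity emitter are disabled.
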NiagaraSystem /BRCosmetics/Characters/Player/Male/Medium/Bodies/M_MED_Gummi/FX/NS_Gummi_Drips
dependencies {
 usedDataInterfaces [
  &NiagaraDataInterfaceSkeletalMesh
  &NiagaraDataInterfaceCurve
  &NiagaraDataInterfaceVector2DCurve
 ]
 usedDynamicInputs []
 usedModules [&SystemState]
}
emitters {
 Drips {
  emitterUpdateScript [
   &EmitterState()
   &SpawnRate(SpawnRate=0.7, Use Spawn Probability=false)
  ]
  particleSpawnScript [
   &InitializeParticle(
    Color Mode=::NewEnumerator0
    Lifetime Max=1.8
    Lifetime Min=1.2
    Lifetime Mode=::NewEnumerator1
    Sprite Size=<2.5 2.5>
    Sprite Size Mode=::NewEnumerator4
    UsePositionOffset=false
    Write Material Random=true
   )
   &FN_SetSpawnGroup(Number of Spawn Groups=8)
   &SkeletalMeshLocation(
    Bone / Socket Index=$Particles.SpawnGroup
    Bone Sampling Mode=::NewEnumerator4
    MaskSpawnGroup=false
    Skeletal Mesh=DataInterface() {propertyValues {
      DefaultMesh=null
      ExcludeBoneName=root
      OptionalComponentClass=null
      PreviewMesh=/BRCosmetics/Characters/Player/Male/Medium/Bodies/M_MED_Gummi/Meshes/M_MED_Gummi
      SkinningMode=SkinOnTheFly
      SoftSourceActor=null
      SourceMode=Default
      UvSetIndex=0
      WholeMeshLOD=-1
      bExcludeBone=false
      bReadDeformedGeometry=true
      bRequireCurrentFrameData=true
      ComponentTags []
      FilteredBones [upperarm_l, upperarm_r, lowerarm_l, lowerarm_r, dyn_calfMeasure_r, dyn_calfMeasure_l, hand_l, hand_r]
      FilteredSockets []
      MeshUserParameter {parameter {name=None, typeDefHandle {registeredTypeIndex=1}, varData []}}
      SamplingRegions []
     }}
   )
  ]
  particleUpdateScript [
   &ParticleState()
   &GravityForce(Gravity=&MakeVector(
     X=0
     Y=0
     Z=&FloatFromCurve(FloatCurve=DataInterface() {propertyValues {Curve {keys [
          {arriveTangent=-3079.7324, interpMode=RCIM_Cubic, leaveTangent=-3079.7324, tangentMode=RCTM_User, time=0, value=-200}
          {arriveTangent=-183.7906, interpMode=RCIM_Cubic, leaveTangent=-183.7906, tangentMode=RCTM_User, time=0.48, value=-880}
         ]}}})
    ))
   &Collision(
    Advanced Aging Rate=100
    Age Colliding Particles=true
    CPU Collision Trace Channel=ECC_GameTraceChannel11
    Module.Control Rest State Via Surface Normal=false
    Module.EnableMaxCollisionCount=false
    Module.Randomize Collision Normal Vector=false
   ) {enabled=false}
   &SolveForcesAndVelocity(Clamp Velocity=false, Limit Acceleration=false)
   &ScaleSpriteSize(
    Scale Factor=&MakeVector2D(
     X=&FloatFromCurve(FloatCurve=DataInterface() {propertyValues {Curve {keys [
          {arriveTangent=-0.1045, interpMode=RCIM_Cubic, leaveTangent=-0.1045, tangentMode=RCTM_User, time=0, value=1}
          {arriveTangent=-1.1433, interpMode=RCIM_Cubic, leaveTangent=-1.1433, tangentMode=RCTM_User, time=1, value=0.55}
         ]}}})
     Y=&FloatFromCurve(FloatCurve=DataInterface() {propertyValues {Curve {keys [
          {arriveTangent=1.8807, interpMode=RCIM_Cubic, leaveTangent=1.8807, tangentMode=RCTM_User, time=0, value=1.8}
          {arriveTangent=10.1826, interpMode=RCIM_Cubic, leaveTangent=10.1826, tangentMode=RCTM_User, time=0.75, value=6}
         ]}}})
    )
    Scale Sprite Size Mode=::NewEnumerator2
   )
   /Niagara/Modules/Events/GenerateCollisionEvent(Module.EventDelay=true, Module.Use Velocity Threshold=true) {enabled=false}
  ]
  rendererClasses [&NiagaraSpriteRendererProperties]
  renderers [{rendererClass=&NiagaraSpriteRendererProperties}]
 }
 Drips_Low {
  simTarget=GPUComputeSim
  emitterUpdateScript [
   &EmitterState()
   &SpawnRate(SpawnRate=0.7, Use Spawn Probability=false)
  ]
  particleSpawnScript [
   &InitializeParticle(
    Color Mode=::NewEnumerator0
    Lifetime Max=1.8
    Lifetime Min=1.2
    Lifetime Mode=::NewEnumerator1
    Sprite Size=<2.5 2.5>
    Sprite Size Mode=::NewEnumerator4
    UsePositionOffset=false
    Write Material Random=true
   )
   &FN_SetSpawnGroup(Number of Spawn Groups=8)
   &SkeletalMeshLocation(
    Bone / Socket Index=$Particles.SpawnGroup
    Bone Sampling Mode=::NewEnumerator4
    MaskSpawnGroup=false
    Skeletal Mesh=DataInterface() {propertyValues {
      DefaultMesh=null
      ExcludeBoneName=root
      OptionalComponentClass=null
      PreviewMesh=&M_MED_Gummi
      SkinningMode=SkinOnTheFly
      SoftSourceActor=null
      SourceMode=Default
      UvSetIndex=0
      WholeMeshLOD=-1
      bExcludeBone=false
      bReadDeformedGeometry=true
      bRequireCurrentFrameData=true
      ComponentTags []
      FilteredBones [upperarm_l, upperarm_r, lowerarm_l, lowerarm_r, dyn_calfMeasure_r, dyn_calfMeasure_l, hand_l, hand_r]
      FilteredSockets []
      MeshUserParameter {parameter {name=None, typeDefHandle {registeredTypeIndex=1}, varData []}}
      SamplingRegions []
     }}
   )
  ]
  particleUpdateScript [
   &ParticleState()
   &GravityForce(Gravity=&MakeVector(
     X=0
     Y=0
     Z=&FloatFromCurve(FloatCurve=DataInterface() {propertyValues {Curve {keys [
          {arriveTangent=-3079.7324, interpMode=RCIM_Cubic, leaveTangent=-3079.7324, tangentMode=RCTM_User, time=0, value=-200}
          {arriveTangent=-183.7906, interpMode=RCIM_Cubic, leaveTangent=-183.7906, tangentMode=RCTM_User, time=0.48, value=-880}
         ]}}})
    ))
   &SolveForcesAndVelocity(Clamp Velocity=false, Limit Acceleration=false)
   &ScaleSpriteSize(
    Scale Factor=&MakeVector2D(
     X=&FloatFromCurve(FloatCurve=DataInterface() {propertyValues {Curve {keys [
          {arriveTangent=-0.1045, interpMode=RCIM_Cubic, leaveTangent=-0.1045, tangentMode=RCTM_User, time=0, value=1}
          {arriveTangent=-1.1433, interpMode=RCIM_Cubic, leaveTangent=-1.1433, tangentMode=RCTM_User, time=1, value=0.55}
         ]}}})
     Y=&FloatFromCurve(FloatCurve=DataInterface() {propertyValues {Curve {keys [
          {arriveTangent=1.8807, interpMode=RCIM_Cubic, leaveTangent=1.8807, tangentMode=RCTM_User, time=0, value=1.8}
          {arriveTangent=10.1826, interpMode=RCIM_Cubic, leaveTangent=10.1826, tangentMode=RCTM_User, time=0.75, value=6}
         ]}}})
    )
    Scale Sprite Size Mode=::NewEnumerator2
   )
  ]
  rendererClasses [&NiagaraSpriteRendererProperties]
  renderers [{rendererClass=&NiagaraSpriteRendererProperties}]
 }
 Empty {
  bEnabled=false
  emitterUpdateScript [&EmitterState()]
  particleSpawnScript [
   &InitializeParticle(
    Color Mode=::NewEnumerator0
    Lifetime Max=1.4
    Lifetime Min=0.6
    Lifetime Mode=::NewEnumerator1
    Sprite Size Max=<3 5>
    Sprite Size Min=<2 4>
    Sprite Size Mode=::NewEnumerator2
    UsePositionOffset=false
    Write Material Random=true
   )
   &AddVelocity(Velocity=&RandomRangeVector(Minimum=<40 40 35>, Maximum=<-40 -40 70>))
  ]
  particleUpdateScript [
   &GravityForce(Gravity=<0 0 -300>)
   &SolveForcesAndVelocity(Clamp Velocity=false, Limit Acceleration=false)
   &ParticleState()
   &ScaleSpriteSizeBySpeed()
  ]
  rendererClasses [&NiagaraSpriteRendererProperties]
  renderers [{rendererClass=&NiagaraSpriteRendererProperties}]
 }
}
summary {
 systemName=NS_Gummi_Drips
 emitters [
  {emitterName=Drips, rendererClasses [&NiagaraSpriteRendererProperties]}
  {bEnabled=false, emitterName=Empty, rendererClasses [&NiagaraSpriteRendererProperties]}
  {emitterName=Drips_Low, simTarget=GPUComputeSim, rendererClasses [&NiagaraSpriteRendererProperties]}
 ]
 userVariables []
}
