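# Physics asset for the M_MED_Gummi male medium player body cosmetic, providing secondary wobble on calves, forearms and pelvis.
# Wobble capsules on each calf and forearm attach to small box base bodies through angularly limited constraints.
# Calf bases chain freely through calf bodies to the pelvis, while forearm bases link straight to the pelvis with free rotation.
# A pelvis bounce sphere swings on a single limited axis, with twist and the other swing locked.
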PhysicsAsset /BRCosmetics/Characters/Player/Male/Medium/Bodies/M_MED_Gummi/Meshes/M_MED_Gummi_Physics
bodies {
 calf_l [{shapeType=Capsule, length=0.505, radius=0.505, rotation=<rot 0 -90 0>}]
 calf_r [{shapeType=Capsule, length=0.505, radius=0.505, rotation=<rot 0 -90 180>}]
 dyn_calfWobble_Base_l [{shapeType=Box, extentX=0.2, extentY=0.2, extentZ=0.2}]
 dyn_calfWobble_Base_r [{shapeType=Box, extentX=0.2, extentY=0.2, extentZ=0.2}]
 dyn_calfWobble_l [{shapeType=Capsule, length=3.9551, radius=5.6076}]
 dyn_calfWobble_r [{shapeType=Capsule, length=3.9551, radius=5.6076}]
 dyn_forearmWobble_Base_l [{shapeType=Box, extentX=1.01, extentY=1.01, extentZ=1.01, rotation=<rot 0 -90 0>}]
 dyn_forearmWobble_Base_r [{shapeType=Box, extentX=1.01, extentY=1.01, extentZ=1.01, rotation=<rot 0 -90 0>}]
 dyn_forearmWobble_l [{shapeType=Capsule, length=2.3965, radius=7.1662}]
 dyn_forearmWobble_r [{shapeType=Capsule, length=2.3965, radius=7.1662}]
 dyn_pelvisBounce [{shapeType=Sphere, radius=13.7046, center=<-1.4659 1.5074 0.6163>}]
 pelvis [{shapeType=Capsule, length=5, radius=5}]
}
constraints [
 {
  bone1Name=dyn_calfWobble_Base_l
  bone2Name=calf_l
  swing1LimitDegrees=45
  swing1Motion=Free
  swing2LimitDegrees=45
  swing2Motion=Free
  twistLimitDegrees=45
  twistMotion=Free
 }
 {
  bone1Name=dyn_calfWobble_Base_r
  bone2Name=calf_r
  swing1LimitDegrees=45
  swing1Motion=Free
  swing2LimitDegrees=45
  swing2Motion=Free
  twistLimitDegrees=45
  twistMotion=Free
 }
 {bone1Name=dyn_calfWobble_r, bone2Name=dyn_calfWobble_Base_r, swing1LimitDegrees=45, swing2LimitDegrees=45, twistLimitDegrees=45}
 {bone1Name=dyn_calfWobble_l, bone2Name=dyn_calfWobble_Base_l, swing1LimitDegrees=45, swing2LimitDegrees=45, twistLimitDegrees=45}
 {
  bone1Name=calf_l
  bone2Name=pelvis
  swing1LimitDegrees=45
  swing1Motion=Free
  swing2LimitDegrees=45
  swing2Motion=Free
  twistLimitDegrees=45
  twistMotion=Free
 }
 {
  bone1Name=calf_r
  bone2Name=pelvis
  swing1LimitDegrees=45
  swing1Motion=Free
  swing2LimitDegrees=45
  swing2Motion=Free
  twistLimitDegrees=45
  twistMotion=Free
 }
 {
  bone1Name=dyn_pelvisBounce
  bone2Name=pelvis
  swing1LimitDegrees=45
  swing1Motion=Locked
  swing2LimitDegrees=15
  twistLimitDegrees=45
  twistMotion=Locked
 }
 {
  bone1Name=dyn_forearmWobble_Base_l
  bone2Name=pelvis
  swing1LimitDegrees=45
  swing1Motion=Free
  swing2LimitDegrees=45
  swing2Motion=Free
  twistLimitDegrees=45
  twistMotion=Free
 }
 {bone1Name=dyn_forearmWobble_l, bone2Name=dyn_forearmWobble_Base_l, swing1LimitDegrees=45, swing2LimitDegrees=45, twistLimitDegrees=45}
 {
  bone1Name=dyn_forearmWobble_Base_r
  bone2Name=pelvis
  swing1LimitDegrees=45
  swing1Motion=Free
  swing2LimitDegrees=45
  swing2Motion=Free
  twistLimitDegrees=45
  twistMotion=Free
 }
 {bone1Name=dyn_forearmWobble_r, bone2Name=dyn_forearmWobble_Base_r, swing1LimitDegrees=45, swing2LimitDegrees=45, twistLimitDegrees=45}
]
