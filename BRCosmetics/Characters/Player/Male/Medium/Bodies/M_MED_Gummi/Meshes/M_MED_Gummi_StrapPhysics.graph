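# Physics asset for the bandolier straps and chest pouch on the male medium Gummi character body.
# Capsules on the pelvis and spine form the torso chain, with capsule and box bodies on the dyn_ strap and pouch bones.
# Tight angular limits govern most strap joints, while spine and upper front-strap joints swing and twist freely.
# The back strap and pouch chains join through the side-back bone, closing a loop between spine_04 and spine_05.
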PhysicsAsset /BRCosmetics/Characters/Player/Male/Medium/Bodies/M_MED_Gummi/Meshes/M_MED_Gummi_StrapPhysics
bodies {
 dyn_bandolierBack_01 [{
   shapeType=Capsule
   length=1.7098
   radius=2.5034
   shapeName=dyn_bandolierBack_01_capsule
   center=<2.0328 0 -0.1764>
   rotation=<rot -86.3555 -179.998 179.999>
  }]
 dyn_bandolierBack_02 [{
   shapeType=Capsule
   length=7.4744
   radius=2.4266
   shapeName=dyn_bandolierBack_02_capsule
   center=<-1.3366 -0.0002 0.0263>
   rotation=<rot -88.8734 -179.9916 179.9983>
  }]
 dyn_bandolierFront_01 [{shapeType=Box, extentX=1.25, extentY=1.5, extentZ=1.13, rotation=<rot 0 90 90>}]
 dyn_bandolierFront_02 [{shapeType=Box, extentX=1.25, extentY=1.5, extentZ=1.13, rotation=<rot 0 90 90>}]
 dyn_bandolierFront_03 [{
   shapeType=Capsule
   length=7.9254
   radius=2.9185
   shapeName=dyn_bandolierFront_03_capsule
   center=<5.0467 1.2869 -0.043>
   rotation=<rot -89.5135 -179.316 179.32>
  }]
 dyn_bandolierSideBack [{
   shapeType=Capsule
   length=4.9731
   radius=2.4564
   shapeName=dyn_bandolierSideBack_capsule
   center=<1.4204 0 0.231>
   rotation=<rot -88.3532 -0.0004 0.0006>
  }]
 dyn_bandolierSideFront [{
   shapeType=Capsule
   length=7.5024
   radius=2.9607
   shapeName=dyn_bandolierSideFront_capsule
   center=<-5.2933 1.3893 0.0828>
   rotation=<rot -89.1041 -179.9913 179.9914>
  }]
 dyn_chestPouch_01 [{
   shapeType=Capsule
   length=27.3661
   radius=4.5223
   shapeName=dyn_chestPouch_01_capsule
   center=<11.9811 0 -0.1117>
   rotation=<rot -85.1647 -179.9972 179.9975>
  }]
 dyn_chestPouch_02 [{
   shapeType=Capsule
   length=8.3292
   radius=8
   shapeName=dyn_chestPouch_02_capsule
   center=<10.4995 -4.582 1.5299>
   rotation=<rot -2.3066 18.9373 22.0462>
  }]
 pelvis [{shapeType=Capsule, length=5, radius=5}]
 spine_03 [{shapeType=Capsule, length=10.2348, radius=11, center=<-6.7696 0.6047 1.3623>, rotation=<rot 0 95 90>}]
 spine_04 [{shapeType=Capsule, length=12.3425, radius=12.1556, center=<-4.7231 2.0244 0.0002>, rotation=<rot 0 92.4063 90>}]
 spine_05 [{shapeType=Capsule, length=14.7838, radius=10.32, center=<-7.5646 0.0001 0>, rotation=<rot 0 90 90>}]
}
constraints [
 {bone1Name=dyn_chestPouch_01, bone2Name=spine_05, swing1LimitDegrees=1, swing2LimitDegrees=14, twistLimitDegrees=1}
 {bone1Name=dyn_chestPouch_02, bone2Name=dyn_chestPouch_01, swing1LimitDegrees=25, swing2LimitDegrees=12, twistLimitDegrees=5}
 {
  bone1Name=spine_05
  bone2Name=spine_04
  swing1LimitDegrees=45
  swing1Motion=Free
  swing2LimitDegrees=45
  swing2Motion=Free
  twistLimitDegrees=45
  twistMotion=Free
 }
 {bone1Name=dyn_bandolierBack_01, bone2Name=spine_04, swing1LimitDegrees=8, swing2LimitDegrees=8, twistLimitDegrees=0}
 {bone1Name=dyn_bandolierBack_02, bone2Name=dyn_bandolierBack_01, swing1LimitDegrees=8, swing2LimitDegrees=2, twistLimitDegrees=1}
 {bone1Name=dyn_bandolierSideBack, bone2Name=dyn_chestPouch_02, swing1LimitDegrees=2, swing2LimitDegrees=25, twistLimitDegrees=1}
 {bone1Name=dyn_bandolierSideFront, bone2Name=dyn_chestPouch_02, swing1LimitDegrees=6, swing2LimitDegrees=6, twistLimitDegrees=4}
 {
  bone1Name=spine_03
  bone2Name=pelvis
  swing1LimitDegrees=45
  swing1Motion=Free
  swing2LimitDegrees=45
  swing2Motion=Free
  twistLimitDegrees=45
  twistMotion=Free
 }
 {
  bone1Name=spine_04
  bone2Name=spine_03
  swing1LimitDegrees=45
  swing1Motion=Free
  swing2LimitDegrees=45
  swing2Motion=Free
  twistLimitDegrees=45
  twistMotion=Free
 }
 {
  bone1Name=dyn_bandolierFront_01
  bone2Name=spine_05
  swing1LimitDegrees=45
  swing1Motion=Free
  swing2LimitDegrees=45
  swing2Motion=Free
  twistLimitDegrees=45
  twistMotion=Free
 }
 {
  bone1Name=dyn_bandolierFront_02
  bone2Name=dyn_bandolierFront_01
  swing1LimitDegrees=45
  swing1Motion=Free
  swing2LimitDegrees=45
  swing2Motion=Free
  twistLimitDegrees=45
  twistMotion=Free
 }
 {bone1Name=dyn_bandolierFront_03, bone2Name=dyn_bandolierFront_02, swing1LimitDegrees=8, swing2LimitDegrees=6, twistLimitDegrees=4}
 {bone1Name=dyn_bandolierBack_02, bone2Name=dyn_bandolierSideBack, swing1LimitDegrees=8, swing2LimitDegrees=8, twistLimitDegrees=1}
]
