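# Animation Blueprint for the M_MED_Gummi player body, adding secondary motion to straps, chest pouch and wobble bones.
# The parent mesh pose passes through bone offsets, strap RigidBody, Control Rig, then a second RigidBody, weakened while crouching.
# Gameplay tags for jump boots and replaced hands enable constraints on calf and forearm wobble bones.
# Look-Ats aim bandolier bones, then sprint, melee, skydive and water-jump states apply corrective offsets.
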AnimBlueprint /BRCosmetics/Characters/Player/Male/Medium/Bodies/M_MED_Gummi/Meshes/M_MED_Gummi_AnimBP
target_skeleton=&/M_MED_Gummi_Skeleton
graphs [
 AnimGraph {nodes [
   "Transform (Modify) Bone - Bone: dyn_bandolierFront_02"(Alpha=1, Translation="0.000,1.5,-0.000", Rotation="0,0,0", Scale="1,1,1") {
    translationMode=BMM_Additive
    alphaBoolBlend {blendOption=Linear}
    alphaScaleBiasClamp {bInterpResult=true}
    boneToModify {boneName=dyn_bandolierFront_02}
    translation {y=1.5}
   }
   n1 = "Constraint - dyn_chestPouch_02 (P)"(ComponentPose=Local To Component(LocalPose=Copy Pose From Mesh() {bUseAttachedParent=true}), Alpha=1, ConstraintWeights_0=1) {
    alphaBoolBlend {blendOption=Linear}
    boneToModify {boneName=dyn_chestPouch_02}
    constraintSetup [{targetBone {boneName=spine_03}}]
    constraintWeights [1]
   }
   n2 = "Transform (Modify) Bone - Bone: dyn_bandolierBack_01"!(ComponentPose=@n1, Alpha=1, Translation="0,-1.147,-0.000") -> (Pose) {translationMode=BMM_Additive, alphaBoolBlend {blendOption=Linear}, boneToModify {boneName=dyn_bandolierBack_01}, translation {y=-1.147}}
   n3 = "Transform (Modify) Bone - Bone: dyn_chestPouch_02"!(ComponentPose=@n2, Alpha=1, Translation="2.008,0.043,-1.702", Rotation="-5,5,5") -> (Pose) {
    rotationMode=BMM_Additive
    rotationSpace=BCS_BoneSpace
    translationMode=BMM_Additive
    translationSpace=BCS_BoneSpace
    alphaBoolBlend {blendOption=Linear}
    boneToModify {boneName=dyn_chestPouch_02}
    rotation=<rot -5 5 5>
    translation=<2.008 0.043 -1.702>
   }
   n4 = "Transform (Modify) Bone - Bone: dyn_bandolierBack_02"!(ComponentPose=@n3, Alpha=1, Translation="0.000,0.787,0.000") -> (Pose) {translationMode=BMM_Additive, alphaBoolBlend {blendOption=Linear}, boneToModify {boneName=dyn_bandolierBack_02}, translation {y=0.787}}
   n5 = "Transform (Modify) Bone - Bone: dyn_bandolierFront_02"!(ComponentPose=@n4, Alpha=1, Translation="0,1,0") -> (Pose) {
    translationMode=BMM_Additive
    translationSpace=BCS_BoneSpace
    alphaBoolBlend {blendOption=Linear}
    boneToModify {boneName=dyn_bandolierFront_02}
    translation {y=1}
   }
   n6 = "Transform (Modify) Bone - Bone: dyn_bandolierFront_03"!(ComponentPose=@n5, Alpha=1, Translation="0,-0.27,0") -> (Pose) {
    translationMode=BMM_Additive
    translationSpace=BCS_BoneSpace
    alphaBoolBlend {blendOption=Linear}
    boneToModify {boneName=dyn_bandolierFront_03}
    translation {y=-0.27}
   }
   n7 = "Transform (Modify) Bone - Bone: dyn_bandolierSideFront"!(ComponentPose=@n6, Alpha=1, Translation="0.000,0.892,0.000") -> (Pose) {translationMode=BMM_Additive, alphaBoolBlend {blendOption=Linear}, boneToModify {boneName=dyn_bandolierSideFront}, translation {y=0.892}}
   n8 = "Transform (Modify) Bone - Bone: dyn_bandolierFront_03"!(ComponentPose=@n7, Alpha=1, Translation="0,3,0", Rotation:Rotator="0,0,0", Scale:Vector="1,1,1") -> (Pose) {
    rotationMode=BMM_Additive
    rotationSpace=BCS_BoneSpace
    translationMode=BMM_Additive
    translationSpace=BCS_BoneSpace
    alphaBoolBlend {blendOption=Linear}
    alphaScaleBiasClamp {bInterpResult=true, interpSpeedDecreasing=20, interpSpeedIncreasing=20}
    boneToModify {boneName=dyn_bandolierFront_03}
    translation {y=3}
   }
   n9 = "Transform (Modify) Bone - Bone: dyn_bandolierFront_02"!(ComponentPose=@n8, Alpha=1, Translation="0,3,0", Rotation:Rotator="-9,0,0", Scale:Vector="1,1,1") -> (Pose) {
    rotationMode=BMM_Additive
    rotationSpace=BCS_BoneSpace
    translationMode=BMM_Additive
    translationSpace=BCS_BoneSpace
    alphaBoolBlend {blendOption=Linear}
    alphaScaleBiasClamp {bInterpResult=true, interpSpeedDecreasing=20, interpSpeedIncreasing=20}
    boneToModify {boneName=dyn_bandolierFront_02}
    rotation {pitch=-9}
    translation {y=3}
   }
   n10 = "Transform (Modify) Bone - Bone: dyn_chestPouch_02"!(ComponentPose=@n9, Alpha=1, Translation="-1,-2,3", Rotation="0,0,0", Scale:Vector="1,1,1") -> (Pose) {
    rotationMode=BMM_Additive
    rotationSpace=BCS_BoneSpace
    translationMode=BMM_Additive
    translationSpace=BCS_BoneSpace
    alphaBoolBlend {blendOption=Linear}
    alphaScaleBiasClamp {bInterpResult=true, interpSpeedDecreasing=30, interpSpeedIncreasing=30}
    boneToModify {boneName=dyn_chestPouch_02}
    translation=<-1 -2 3>
   }
   n11 = RigidBody(ComponentPose=@n10) {
    bClampLinearTranslationLimitToRefPose=true
    bOverrideWorldGravity=true
    overridePhysicsAsset=&/M_MED_Gummi_StrapPhysics
    simulationSpace=BaseBoneSpace
    alphaBoolBlend {blendOption=Linear}
    baseBoneRef {boneName=spine_03}
    componentAppliedLinearAccClamp=<6000 6000 6000>
    componentLinearAccScale=<0.3 0.3 0.3>
    componentLinearVelScale=<0.3 0.3 0.3>
    overrideWorldGravity {z=-980}
    simSpaceSettings {worldAlpha=0.1}
   }
   n12 = Control Rig(Source=Component To Local(ComponentPose=@n11)) {
    alphaBoolBlend {blendOption=Linear}
    controlRigAssetReference {blueprintRigClass=&/M_MED_Gummi_Body_CtrlRig.M_MED_Gummi_Body_CtrlRig_C}
   }
   n13 = RigidBody(ComponentPose=Local To Component(LocalPose=@n12), Alpha=Get bounceAlpha() -> (bounceAlpha:double)) {
    bClampLinearTranslationLimitToRefPose=true
    bOverrideWorldGravity=true
    overridePhysicsAsset=&/M_MED_Gummi_Physics
    simulationSpace=BaseBoneSpace
    alphaBoolBlend {blendOption=Linear}
    baseBoneRef {boneName=root}
    componentAppliedLinearAccClamp=<6000 6000 6000>
    componentLinearAccScale=<3 3 5>
    componentLinearVelScale=<0.2 0.2 0.54>
    overrideWorldGravity {z=-980}
   }
   n14 = Get is_wearing_jumpboots() -> (is_wearing_jumpboots:bool)
   n15 = "Constraint - dyn_calfWobble_l (P)"!(ComponentPose=@n13, bAlphaBoolEnabled:bool=@n14, ConstraintWeights_0=1) -> (Pose) {
    alphaInputType=Bool
    alphaBoolBlend {blendOption=Linear}
    boneToModify {boneName=dyn_calfWobble_l}
    constraintSetup [{targetBone {boneName=dyn_calfWobble_Base_l}}]
    constraintWeights [1]
   }
   n16 = "Constraint - dyn_calfWobble_r (P)"!(ComponentPose=@n15, bAlphaBoolEnabled:bool=@n14, ConstraintWeights_0=1) -> (Pose) {
    alphaInputType=Bool
    alphaBoolBlend {blendOption=Linear}
    boneToModify {boneName=dyn_calfWobble_r}
    constraintSetup [{targetBone {boneName=dyn_calfWobble_Base_r}}]
    constraintWeights [1]
   }
   n17 = Get is_wearing_westsausage() -> (is_wearing_westsausage:bool)
   n18 = "Constraint - dyn_forearmWobble_l (P)"!(ComponentPose=@n16, bAlphaBoolEnabled:bool=@n17, ConstraintWeights_0=1) -> (Pose) {
    alphaInputType=Bool
    alphaBoolBlend {blendOption=Linear}
    boneToModify {boneName=dyn_forearmWobble_l}
    constraintSetup [{targetBone {boneName=dyn_forearmWobble_Base_l}}]
    constraintWeights [1]
   }
   n19 = "Constraint - dyn_forearmWobble_r (P)"!(ComponentPose=@n18, bAlphaBoolEnabled:bool=@n17, ConstraintWeights_0=1) -> (Pose) {
    alphaInputType=Bool
    alphaBoolBlend {blendOption=Linear}
    boneToModify {boneName=dyn_forearmWobble_r}
    constraintSetup [{targetBone {boneName=dyn_forearmWobble_Base_r}}]
    constraintWeights [1]
   }
   n20 = Look At - Bone: dyn_bandolierBack_01(ComponentPose=@n19, Alpha=1) {
    interpolationTime=0.8
    lookAtClamp=10
    alphaBoolBlend {blendOption=Linear}
    boneToModify {boneName=dyn_bandolierBack_01}
    lookAtLocation=<0 -2 0>
    lookAtTarget {boneReference {boneName=dyn_bandolierSideBack}}
    lookAt_Axis {axis=<1 0>}
    lookUp_Axis {axis {x=0, z=-1}}
   }
   n21 = Look At - Bone: dyn_bandolierFront_03(ComponentPose=@n20, Alpha=1) {
    interpolationTime=0.8
    lookAtClamp=10
    alphaBoolBlend {blendOption=Linear}
    boneToModify {boneName=dyn_bandolierFront_03}
    lookAtLocation {x=0, z=-10}
    lookAtTarget {boneReference {boneName=dyn_chestPouchEND}}
    lookAt_Axis {axis=<1 0>}
    lookUp_Axis {axis {x=0, z=-1}}
   }
   n22 = Look At - Bone: dyn_bandolierFront_01(ComponentPose=@n21, Alpha=0.5) {
    interpolationTime=0.8
    lookAtClamp=10
    alphaBoolBlend {blendOption=Linear}
    boneToModify {boneName=dyn_bandolierFront_01}
    lookAtLocation=<0 15 5>
    lookAtTarget {boneReference {boneName=dyn_bandolierSideFront}}
    lookAt_Axis {axis=<1 0>}
    lookUp_Axis {axis {x=0, z=-1}}
   }
   n23 = Look At - Bone: dyn_bandolierFront_02(ComponentPose=@n22, Alpha=0.5) {
    interpolationTime=0.8
    lookAtClamp=10
    alphaBoolBlend {blendOption=Linear}
    boneToModify {boneName=dyn_bandolierFront_02}
    lookAtLocation=<0 10>
    lookAtTarget {boneReference {boneName=dyn_bandolierSideFront}}
    lookAt_Axis {axis=<1 0>}
    lookUp_Axis {axis {x=0, z=-1}}
   }
   n24 = Get Is Sprinting() -> (Is Sprinting:double)
   n25 = "Transform (Modify) Bone - Bone: dyn_bandolierBack_01"!(ComponentPose=@n23, Alpha=@n24, Translation="0.000,-4,-0.000", Rotation:Rotator="0,0,0", Scale:Vector="1,1,1") -> (Pose) {
    translationMode=BMM_Additive
    alphaBoolBlend {blendOption=Linear}
    alphaScaleBiasClamp {bInterpResult=true}
    boneToModify {boneName=dyn_bandolierBack_01}
    translation {y=-4}
   }
   n26 = "Transform (Modify) Bone - Bone: dyn_chestPouch_02"(ComponentPose=@n25, Alpha=@n24, Translation="0.000,-2,-0.000", Rotation="0,0,0", Scale="1,1,1") {
    translationMode=BMM_Additive
    alphaBoolBlend {blendOption=Linear}
    alphaScaleBiasClamp {bInterpResult=true}
    boneToModify {boneName=dyn_chestPouch_02}
    translation {y=-2}
   }
   n27 = Get Is Playing Melee Anim() -> (Is Playing Melee Anim:double)
   n28 = "Transform (Modify) Bone - Bone: dyn_bandolierFront_01"(ComponentPose=@n26, Alpha=@n27, Translation="0.000,1.5,-0.000", Rotation="0,0,0", Scale="1,1,1") {
    translationMode=BMM_Additive
    alphaBoolBlend {blendOption=Linear}
    alphaScaleBiasClamp {bInterpResult=true}
    boneToModify {boneName=dyn_bandolierFront_01}
    translation {y=1.5}
   }
   n29 = "Transform (Modify) Bone - Bone: dyn_bandolierFront_02"(ComponentPose=@n28, Alpha=@n27, Translation="-1,1.5,0", Rotation="0,0,0", Scale="1,1,1") {
    translationMode=BMM_Additive
    alphaBoolBlend {blendOption=Linear}
    alphaScaleBiasClamp {bInterpResult=true}
    boneToModify {boneName=dyn_bandolierFront_02}
    translation=<-1 1.5>
   }
   n30 = "Transform (Modify) Bone - Bone: dyn_bandolierFront_03"(ComponentPose=@n29, Alpha=@n27, Translation="0,1,0", Rotation="0,0,0", Scale="1,1,1") {
    translationMode=BMM_Additive
    alphaBoolBlend {blendOption=Linear}
    alphaScaleBiasClamp {bInterpResult=true}
    boneToModify {boneName=dyn_bandolierFront_03}
    translation {y=1}
   }
   n31 = "Transform (Modify) Bone - Bone: dyn_bandolierBack_01"(ComponentPose=@n30, Alpha=@n27, Translation="0,-1.257,0", Rotation="0,0,0", Scale="1,1,1") {
    rotationMode=BMM_Additive
    translationMode=BMM_Additive
    alphaBoolBlend {blendOption=Linear}
    alphaScaleBiasClamp {bInterpResult=true}
    boneToModify {boneName=dyn_bandolierBack_01}
    translation {y=-1.257}
   }
   n32 = "Transform (Modify) Bone - Bone: dyn_chestPouch_01"(ComponentPose=@n31, Alpha=@n27, Translation="0,-2,0", Rotation="0,0,0", Scale="1,1,1") {
    rotationMode=BMM_Additive
    translationMode=BMM_Additive
    alphaBoolBlend {blendOption=Linear}
    alphaScaleBiasClamp {bInterpResult=true}
    boneToModify {boneName=dyn_chestPouch_01}
    translation {y=-2}
   }
   n33 = Get Is Skydive Dive Mode() -> (Is Skydive Dive Mode:double)
   n34 = "Transform (Modify) Bone - Bone: dyn_bandolierBack_02"!(ComponentPose=@n32, Alpha=@n33, Translation="-0.000,3,8", Rotation:Rotator="0,0,0", Scale:Vector="1,1,1") -> (Pose) {
    translationMode=BMM_Additive
    alphaBoolBlend {blendOption=Linear}
    alphaScaleBiasClamp {bInterpResult=true}
    boneToModify {boneName=dyn_bandolierBack_02}
    translation {y=3, z=8}
   }
   n35 = "Transform (Modify) Bone - Bone: dyn_bandolierBack_01"(ComponentPose=@n34, Alpha=@n33, Translation="0.000,3,8", Rotation="0,0,0", Scale="1,1,1") {
    translationMode=BMM_Additive
    alphaBoolBlend {blendOption=Linear}
    alphaScaleBiasClamp {bInterpResult=true}
    boneToModify {boneName=dyn_bandolierBack_01}
    translation {y=3, z=8}
   }
   n36 = "Transform (Modify) Bone - Bone: dyn_chestPouch_02"(ComponentPose=@n35, Alpha=@n33, Translation="-4,3.000,5.000", Rotation="0,0,0", Scale="1,1,1") {
    translationMode=BMM_Additive
    alphaBoolBlend {blendOption=Linear}
    alphaScaleBiasClamp {bInterpResult=true}
    boneToModify {boneName=dyn_chestPouch_02}
    translation=<-4 3 5>
   }
   n37 = Get Is Skydive Floating() -> (Is Skydive Floating:double)
   n38 = "Transform (Modify) Bone - Bone: dyn_bandolierBack_01"(ComponentPose=@n36, Alpha=@n37, Translation="0.000,-8,2", Rotation="0,0,0", Scale="1,1,1") {
    translationMode=BMM_Additive
    alphaBoolBlend {blendOption=Linear}
    alphaScaleBiasClamp {bInterpResult=true}
    boneToModify {boneName=dyn_bandolierBack_01}
    translation {y=-8, z=2}
   }
   n39 = "Transform (Modify) Bone - Bone: dyn_bandolierBack_02"(ComponentPose=@n38, Alpha=@n37, Translation="-2,-5,0", Rotation="0,0,0", Scale="1,1,1") {
    translationMode=BMM_Additive
    alphaBoolBlend {blendOption=Linear}
    alphaScaleBiasClamp {bInterpResult=true}
    boneToModify {boneName=dyn_bandolierBack_02}
    translation=<-2 -5>
   }
   n40 = "Transform (Modify) Bone - Bone: dyn_chestPouch_02"(ComponentPose=@n39, Alpha=@n37, Translation="-3,-5.000,0.000", Rotation="0,0,0", Scale="1,1,1") {
    translationMode=BMM_Additive
    alphaBoolBlend {blendOption=Linear}
    alphaScaleBiasClamp {bInterpResult=true}
    boneToModify {boneName=dyn_chestPouch_02}
    translation=<-3 -5>
   }
   n41 = "Transform (Modify) Bone - Bone: dyn_chestPouch_02"!(ComponentPose=@n40, bAlphaBoolEnabled:bool=Get bIsSprinting(), Translation="0,0,3", Rotation="0,0,0", Scale="1,1,1") -> (Pose) {
    alphaInputType=Bool
    rotationMode=BMM_Additive
    rotationSpace=BCS_BoneSpace
    translationMode=BMM_Additive
    translationSpace=BCS_BoneSpace
    alphaBoolBlend {blendInTime=0.1, blendOption=Linear, blendOutTime=0.2}
    alphaScaleBiasClamp {bInterpResult=true}
    boneToModify {boneName=dyn_chestPouch_02}
    translation {z=3}
   }
   n42 = "Transform (Modify) Bone - Bone: dyn_bandolierBack_01"!(
    ComponentPose=@n41
    bAlphaBoolEnabled:bool=Get bIsWaterJump() -> (bIsWaterJump:bool)
    Translation="-0.000,-4.112,-1.859"
    Rotation="0.005,8.710,-0.002"
    Scale="1,1,1"
   ) -> (Pose) {
    alphaInputType=Bool
    rotationMode=BMM_Additive
    rotationSpace=BCS_BoneSpace
    translationMode=BMM_Additive
    translationSpace=BCS_BoneSpace
    alphaBoolBlend {blendInTime=0.1, blendOption=Linear, blendOutTime=0.2}
    alphaScaleBiasClamp {bInterpResult=true}
    boneToModify {boneName=dyn_bandolierBack_01}
    rotation=<rot 0.005 8.71 -0.002>
    translation {y=-4.112, z=-1.859}
   }
   Output Pose(Result=Component To Local(ComponentPose=@n42)) {layerGroup=None}
  ]}
 EventGraph {nodes [
   Delay(Duration=2)
   > Set bounceAlpha(bounceAlpha:double=1) -> (Output_Get:double)
   n1 = Event BlueprintUpdateAnimation()
   > n2 = Sequence!(execute) -> (then_0, then_1, then_2:Exec, then_3:Exec, then_4:Exec)
   > n3 = Branch(Condition=Get bIsCrouching())
   > Set bounceAlpha(bounceAlpha=FInterpTo_Constant!(Current:double=Get bounceAlpha(), Target:double=0.35, DeltaTime:double=@n1.DeltaTimeX, InterpSpeed:double=5) -> (ReturnValue:double))
   Set bounceAlpha(execute=@n3.else, bounceAlpha=FInterpTo_Constant(Current=Get bounceAlpha(), Target=1, DeltaTime=@n1.DeltaTimeX, InterpSpeed=1))
   Set Is Sprinting(execute=@n2.then_4, Is Sprinting:double="To Float (Boolean)"(InBool=Get bIsSprinting())) -> (Output_Get:double)
   > Set Is Playing Melee Anim(Is Playing Melee Anim:double="To Float (Boolean)"(InBool=Get bIsPlayingMeleeAnim())) -> (Output_Get:double)
   > Set Is Skydive Dive Mode(Is Skydive Dive Mode:double="To Float (Boolean)"(InBool=Get bIsSkydiveDiveMode())) -> (Output_Get:double)
   > Set Is Skydive Floating(Is Skydive Floating:double="To Float (Boolean)"(InBool=Get bIsSkydiveFloating())) -> (Output_Get:double)
   > n4 = Branch(Condition=OR Boolean(A=Get bIsPlayingEmote(), B=Get bIsSkyDiving()))
   > Set Bag Dynamics Alpha(Bag Dynamics Alpha:double=0.5) -> (Output_Get:double)
   Set Bag Dynamics Alpha(execute=@n4.else, Bag Dynamics Alpha=1)
   n5 = Cast To FortPlayerPawn(execute=@n2.then_3, Object=GetOwningActor())
   > Set As Fort Player Pawn(As Fort Player Pawn:Fort Player Pawn Object Reference=@n5) -> (Output_Get:Fort Player Pawn Object Reference)
   > n6 = Get!(execute:Exec) -> (then:Exec, else:Exec, As Fort Player Pawn:Fort Player Pawn Object Reference)
   > Set is_wearing_westsausage(is_wearing_westsausage:bool=HasMatchingGameplayTag!(self:Gameplay Tag Asset Interface Interface=@n6, TagToCheck:Gameplay Tag Structure=(TagName="Granted.Athena.HandsAreReplaced")) -> (ReturnValue:bool)) -> (Output_Get:bool)
   n7 = Cast To FortPlayerPawn(execute=@n2.then_1, Object=GetOwningActor())
   > Set As Fort Player Pawn(As Fort Player Pawn=@n7)
   > n8 = Get()
   > Set is_wearing_jumpboots(is_wearing_jumpboots:bool=HasMatchingGameplayTag(self=@n8, TagToCheck=(TagName="Item.Gadget.JumpBoots.AnimTracker"))) -> (Output_Get:bool)
   n9 = GetAnimInstance(self=GetOwnerSkeletalMeshComponent())
   Is Valid(exec=@n2.then_2, InputObject=@n9)
   n10 = Montage_IsPlaying(self=@n9, Montage=/Game/Animation/Game/MainPlayer/Montages/Emotes/Emote_DanceMoves)
   n11 = Montage_IsPlaying(self=@n9, Montage=/BRCosmetics/Animation/Game/MainPlayer/Emotes/ElectroSwing/Electroswing_CMM_M)
   n12 = Montage_IsPlaying(self=@n9, Montage=/BRCosmetics/Animation/Game/MainPlayer/Emotes/Walkywalk/Emote_Walkywalk_CMM_M)
   n13 = Montage_IsPlaying(self=@n9, Montage=/BRCosmetics/Animation/Game/MainPlayer/Emotes/Twist/Emote_Twist_CMM_M)
   > Set isCustomEmotePlaying(isCustomEmotePlaying:double=SelectFloat(A=1, B=0, bPickA=OR Boolean!(A=@n12, B=@n13, C:bool=@n11, D:bool=@n10, E:bool) -> (ReturnValue))) -> (Output_Get:double)
   > Set isCustomEmotePlaying2(isCustomEmotePlaying2:double=SelectFloat(A=1, B=0, bPickA=OR Boolean(A=@n10, B=@n12))) -> (Output_Get:double)
  ]}
]
variables [
 As Fort Player Pawn {type=FortPlayerPawn}
 Bag Dynamics Alpha {type=number}
 bounceAlpha {type=number, default=1}
 Is Playing Melee Anim {type=number}
 Is Skydive Dive Mode {type=number}
 Is Skydive Floating {type=number}
 Is Sprinting {type=number}
 is_wearing_jumpboots {type=boolean}
 is_wearing_westsausage {type=boolean}
 isCustomEmotePlaying {type=number}
 isCustomEmotePlaying2 {type=number}
]
