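# Control Rig for the Gummi character's head part, adding procedural motion to its "squiggle" bone chains and adjusting shoulder helper bones.
# Clavicle-driven parent constraints feed shoulder-float SetTransforms flagged initial, which mix axes from float and clavicle-end bones differently per side.
# Sixteen ChainHarmonics units then add pendulum and varied sine-wave wobble, with debug drawing enabled; squiggle C has none.
# Older constraint and upper-arm nodes remain disconnected from execution.
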ControlRigBlueprint /BRCosmetics/Characters/Player/Male/Medium/Bodies/M_MED_Gummi/Meshes/Parts/M_MED_Gummi_Head_CtrlRig
graphs [
 RigVMFunctionLibrary {links [], nodes []}
 RigVMModel {nodes [
   n1 = &RigUnit_BeginExecution()
   &RigUnit_ParentConstraint(
    AdvancedSettings=(RotationOrderForFilter=XZY)
    Child=(Type=Bone,Name="dyn_centerClav_r")
    Parents=((Item=(Type=Bone,Name="clavicle_r"),Weight=1))
    Weight=1
    ExecutePin=@n1.ExecutePin
   )
   > &RigUnit_ParentConstraint(
    AdvancedSettings=(RotationOrderForFilter=XZY)
    Child=(Type=Bone,Name="dyn_centerClav_l")
    Parents=((Item=(Type=Bone,Name="clavicle_l"),Weight=1))
    Weight=1
   )
   n2 = &RigUnit_GetTransform(Item=(Type=Bone,Name="dyn_centerClavEND_r"))
   > &RigUnit_SetTransform(
    Item=(Type=Bone,Name="dyn_shoulderFloat_r")
    Value=(Rotation=(X=0,Y=0,Z=0,W=1),Translation=(X=0,Y=0,Z=0),Scale3D=(X=1,Y=1,Z=1))
    Weight=1
    bInitial=true
    Value.Translation.X=&RigUnit_GetTransform(Item=(Type=Bone,Name="dyn_shoulderFloat_r")).Transform.Translation.X
    Value.Translation.Y=@n2.Transform.Translation.Y
    Value.Translation.Z=@n2.Transform.Translation.Z
   )
   n3 = &RigUnit_GetTransform(Item=(Type=Bone,Name="dyn_shoulderFloat_l"), bInitial=false)
   > &RigUnit_SetTransform(
    Item=(Type=Bone,Name="dyn_shoulderFloat_l")
    Value=(Rotation=(X=0,Y=0,Z=0,W=1),Translation=(X=0,Y=0,Z=0),Scale3D=(X=1,Y=1,Z=1))
    Weight=1
    bInitial=true
    Value.Translation.X=&RigUnit_GetTransform(Item=(Type=Bone,Name="dyn_centerClavEND_l"), bInitial=false).Transform.Translation.X
    Value.Translation.Y=@n3.Transform.Translation.Y
    Value.Translation.Z=@n3.Transform.Translation.Z
   )
   > /Script/ControlRig.RigUnit_ChainHarmonics(
    ChainRoot=dyn_squiggleA_01_Base
    DrawWorldOffset=(Rotation=(X=0,Y=0,Z=0,W=1),Translation=(X=0,Y=0,Z=0),Scale3D=(X=1,Y=1,Z=1))
    Pendulum=(bEnabled=True,PendulumStiffness=2,PendulumGravity=(X=0,Y=0,Z=0),PendulumBlend=0.75,PendulumDrag=0.98,PendulumMinimum=0,PendulumMaximum=0.1,PendulumEase=Linear,UnwindAxis=(X=0,Y=1,Z=0),UnwindMinimum=0.2,UnwindMaximum=0.05)
    Reach=(bEnabled=False,ReachTarget=(X=0,Y=0,Z=0),ReachAxis=(X=1,Y=0,Z=0),ReachMinimum=0,ReachMaximum=0,ReachEase=Linear)
    Speed=(X=1,Y=1,Z=1)
    Wave=(bEnabled=True,WaveFrequency=(X=100,Y=100,Z=100),WaveAmplitude=(X=0,Y=0.5,Z=0.5),WaveOffset=(X=0,Y=1,Z=2),WaveNoise=(X=0,Y=0,Z=0),WaveMinimum=0,WaveMaximum=0.4,WaveEase=SineEaseOut)
    WaveCurve=(EditorCurveData=(Keys=((),(Time=1,Value=1)),PreInfinityExtrap=RCCE_Constant,PostInfinityExtrap=RCCE_Constant,DefaultValue=340282346638528859811704183484516925440))
    bDrawDebug=True
   ) {is=struct RigVMUnitNode}
   > &RigUnit_ChainHarmonics(
    ChainRoot=dyn_squiggleB_01_Base
    DrawWorldOffset=(Rotation=(X=0,Y=0,Z=0,W=1),Translation=(X=0,Y=0,Z=0),Scale3D=(X=1,Y=1,Z=1))
    Pendulum=(bEnabled=True,PendulumStiffness=2,PendulumGravity=(X=0,Y=0,Z=0),PendulumBlend=0.75,PendulumDrag=0.98,PendulumMinimum=0,PendulumMaximum=0.1,PendulumEase=Linear,UnwindAxis=(X=0,Y=1,Z=0),UnwindMinimum=0.2,UnwindMaximum=0.05)
    Reach=(bEnabled=False,ReachTarget=(X=0,Y=0,Z=0),ReachAxis=(X=1,Y=0,Z=0),ReachMinimum=0,ReachMaximum=0,ReachEase=Linear)
    Speed=(X=1,Y=1,Z=1)
    Wave=(bEnabled=True,WaveFrequency=(X=105,Y=105,Z=105),WaveAmplitude=(X=0,Y=0.5,Z=0.5),WaveOffset=(X=0,Y=1,Z=2),WaveNoise=(X=0,Y=0,Z=0),WaveMinimum=0,WaveMaximum=0.4,WaveEase=SineEaseOut)
    WaveCurve=(EditorCurveData=(Keys=((),(Time=1,Value=1)),PreInfinityExtrap=RCCE_Constant,PostInfinityExtrap=RCCE_Constant,DefaultValue=340282346638528859811704183484516925440))
    bDrawDebug=True
   )
   > &RigUnit_ChainHarmonics(
    ChainRoot=dyn_squiggleD_01_Base
    DrawWorldOffset=(Rotation=(X=0,Y=0,Z=0,W=1),Translation=(X=0,Y=0,Z=0),Scale3D=(X=1,Y=1,Z=1))
    Pendulum=(bEnabled=True,PendulumStiffness=2,PendulumGravity=(X=0,Y=0,Z=0),PendulumBlend=0.75,PendulumDrag=0.98,PendulumMinimum=0,PendulumMaximum=0.1,PendulumEase=Linear,UnwindAxis=(X=0,Y=1,Z=0),UnwindMinimum=0.2,UnwindMaximum=0.05)
    Reach=(bEnabled=False,ReachTarget=(X=0,Y=0,Z=0),ReachAxis=(X=1,Y=0,Z=0),ReachMinimum=0,ReachMaximum=0,ReachEase=Linear)
    Speed=(X=1,Y=1,Z=1)
    Wave=(bEnabled=True,WaveFrequency=(X=95,Y=95,Z=95),WaveAmplitude=(X=0,Y=0.25,Z=0.25),WaveOffset=(X=0,Y=0.5,Z=0.25),WaveNoise=(X=0,Y=0,Z=0),WaveMinimum=0,WaveMaximum=0.4,WaveEase=SineEaseOut)
    WaveCurve=(EditorCurveData=(Keys=((),(Time=1,Value=1)),PreInfinityExtrap=RCCE_Constant,PostInfinityExtrap=RCCE_Constant,DefaultValue=340282346638528859811704183484516925440))
    bDrawDebug=True
   )
   > &RigUnit_ChainHarmonics(
    ChainRoot=dyn_squiggleE_01_Base
    DrawWorldOffset=(Rotation=(X=0,Y=0,Z=0,W=1),Translation=(X=0,Y=0,Z=0),Scale3D=(X=1,Y=1,Z=1))
    Pendulum=(bEnabled=True,PendulumStiffness=2,PendulumGravity=(X=0,Y=0,Z=0),PendulumBlend=0.75,PendulumDrag=0.98,PendulumMinimum=0,PendulumMaximum=0.1,PendulumEase=Linear,UnwindAxis=(X=0,Y=1,Z=0),UnwindMinimum=0.2,UnwindMaximum=0.05)
    Reach=(bEnabled=False,ReachTarget=(X=0,Y=0,Z=0),ReachAxis=(X=1,Y=0,Z=0),ReachMinimum=0,ReachMaximum=0,ReachEase=Linear)
    Speed=(X=1,Y=1,Z=1)
    Wave=(bEnabled=True,WaveFrequency=(X=90,Y=90,Z=90),WaveAmplitude=(X=0,Y=0.3,Z=0.15),WaveOffset=(X=0,Y=0.5,Z=0.6),WaveNoise=(X=0,Y=0,Z=0),WaveMinimum=0,WaveMaximum=0.4,WaveEase=SineEaseOut)
    WaveCurve=(EditorCurveData=(Keys=((),(Time=1,Value=1)),PreInfinityExtrap=RCCE_Constant,PostInfinityExtrap=RCCE_Constant,DefaultValue=340282346638528859811704183484516925440))
    bDrawDebug=True
   )
   > &RigUnit_ChainHarmonics(
    ChainRoot=dyn_squiggleF_01_Base
    DrawWorldOffset=(Rotation=(X=0,Y=0,Z=0,W=1),Translation=(X=0,Y=0,Z=0),Scale3D=(X=1,Y=1,Z=1))
    Pendulum=(bEnabled=True,PendulumStiffness=2,PendulumGravity=(X=0,Y=0,Z=0),PendulumBlend=0.75,PendulumDrag=0.98,PendulumMinimum=0,PendulumMaximum=0.1,PendulumEase=Linear,UnwindAxis=(X=0,Y=1,Z=0),UnwindMinimum=0.2,UnwindMaximum=0.05)
    Reach=(bEnabled=False,ReachTarget=(X=0,Y=0,Z=0),ReachAxis=(X=1,Y=0,Z=0),ReachMinimum=0,ReachMaximum=0,ReachEase=Linear)
    Speed=(X=1,Y=1,Z=1)
    Wave=(bEnabled=True,WaveFrequency=(X=108,Y=108,Z=108),WaveAmplitude=(X=0,Y=0.5,Z=0.5),WaveOffset=(X=0,Y=1,Z=2),WaveNoise=(X=0,Y=0,Z=0),WaveMinimum=0,WaveMaximum=0.4,WaveEase=SineEaseOut)
    WaveCurve=(EditorCurveData=(Keys=((),(Time=1,Value=1)),PreInfinityExtrap=RCCE_Constant,PostInfinityExtrap=RCCE_Constant,DefaultValue=340282346638528859811704183484516925440))
    bDrawDebug=True
   )
   > &RigUnit_ChainHarmonics(
    ChainRoot=dyn_squiggleG_01_Base
    DrawWorldOffset=(Rotation=(X=0,Y=0,Z=0,W=1),Translation=(X=0,Y=0,Z=0),Scale3D=(X=1,Y=1,Z=1))
    Pendulum=(bEnabled=True,PendulumStiffness=2,PendulumGravity=(X=0,Y=0,Z=0),PendulumBlend=0.75,PendulumDrag=0.98,PendulumMinimum=0,PendulumMaximum=0.1,PendulumEase=Linear,UnwindAxis=(X=0,Y=1,Z=0),UnwindMinimum=0.2,UnwindMaximum=0.05)
    Reach=(bEnabled=False,ReachTarget=(X=0,Y=0,Z=0),ReachAxis=(X=1,Y=0,Z=0),ReachMinimum=0,ReachMaximum=0,ReachEase=Linear)
    Speed=(X=1,Y=1,Z=1)
    Wave=(bEnabled=True,WaveFrequency=(X=102,Y=102,Z=102),WaveAmplitude=(X=0,Y=0.5,Z=0.5),WaveOffset=(X=0,Y=1,Z=2),WaveNoise=(X=0,Y=0,Z=0),WaveMinimum=0,WaveMaximum=0.4,WaveEase=SineEaseOut)
    WaveCurve=(EditorCurveData=(Keys=((),(Time=1,Value=1)),PreInfinityExtrap=RCCE_Constant,PostInfinityExtrap=RCCE_Constant,DefaultValue=340282346638528859811704183484516925440))
    bDrawDebug=True
   )
   > &RigUnit_ChainHarmonics(
    ChainRoot=dyn_squiggleH_01_Base
    DrawWorldOffset=(Rotation=(X=0,Y=0,Z=0,W=1),Translation=(X=0,Y=0,Z=0),Scale3D=(X=1,Y=1,Z=1))
    Pendulum=(bEnabled=True,PendulumStiffness=2,PendulumGravity=(X=0,Y=0,Z=0),PendulumBlend=0.75,PendulumDrag=0.98,PendulumMinimum=0,PendulumMaximum=0.1,PendulumEase=Linear,UnwindAxis=(X=0,Y=1,Z=0),UnwindMinimum=0.2,UnwindMaximum=0.05)
    Reach=(bEnabled=False,ReachTarget=(X=0,Y=0,Z=0),ReachAxis=(X=1,Y=0,Z=0),ReachMinimum=0,ReachMaximum=0,ReachEase=Linear)
    Speed=(X=1,Y=1,Z=1)
    Wave=(bEnabled=True,WaveFrequency=(X=101,Y=101,Z=101),WaveAmplitude=(X=0,Y=0.4,Z=0.4),WaveOffset=(X=0,Y=1,Z=2),WaveNoise=(X=0,Y=0,Z=0),WaveMinimum=0,WaveMaximum=0.4,WaveEase=SineEaseOut)
    WaveCurve=(EditorCurveData=(Keys=((),(Time=1,Value=1)),PreInfinityExtrap=RCCE_Constant,PostInfinityExtrap=RCCE_Constant,DefaultValue=340282346638528859811704183484516925440))
    bDrawDebug=True
   )
   > &RigUnit_ChainHarmonics(
    ChainRoot=dyn_squiggleJ_01_Base
    DrawWorldOffset=(Rotation=(X=0,Y=0,Z=0,W=1),Translation=(X=0,Y=0,Z=0),Scale3D=(X=1,Y=1,Z=1))
    Pendulum=(bEnabled=True,PendulumStiffness=2,PendulumGravity=(X=0,Y=0,Z=0),PendulumBlend=0.75,PendulumDrag=0.98,PendulumMinimum=0,PendulumMaximum=0.1,PendulumEase=Linear,UnwindAxis=(X=0,Y=1,Z=0),UnwindMinimum=0.2,UnwindMaximum=0.05)
    Reach=(bEnabled=False,ReachTarget=(X=0,Y=0,Z=0),ReachAxis=(X=1,Y=0,Z=0),ReachMinimum=0,ReachMaximum=0,ReachEase=Linear)
    Speed=(X=1,Y=1,Z=1)
    Wave=(bEnabled=True,WaveFrequency=(X=105,Y=105,Z=105),WaveAmplitude=(X=0,Y=0.5,Z=0.5),WaveOffset=(X=0,Y=1,Z=2),WaveNoise=(X=0,Y=0,Z=0),WaveMinimum=0,WaveMaximum=0.4,WaveEase=SineEaseOut)
    WaveCurve=(EditorCurveData=(Keys=((),(Time=1,Value=1)),PreInfinityExtrap=RCCE_Constant,PostInfinityExtrap=RCCE_Constant,DefaultValue=340282346638528859811704183484516925440))
    bDrawDebug=True
   )
   > &RigUnit_ChainHarmonics(
    ChainRoot=dyn_squiggleK_01_Base
    DrawWorldOffset=(Rotation=(X=0,Y=0,Z=0,W=1),Translation=(X=0,Y=0,Z=0),Scale3D=(X=1,Y=1,Z=1))
    Pendulum=(bEnabled=True,PendulumStiffness=2,PendulumGravity=(X=0,Y=0,Z=0),PendulumBlend=0.75,PendulumDrag=0.98,PendulumMinimum=0,PendulumMaximum=0.1,PendulumEase=Linear,UnwindAxis=(X=0,Y=1,Z=0),UnwindMinimum=0.2,UnwindMaximum=0.05)
    Reach=(bEnabled=False,ReachTarget=(X=0,Y=0,Z=0),ReachAxis=(X=1,Y=0,Z=0),ReachMinimum=0,ReachMaximum=0,ReachEase=Linear)
    Speed=(X=1,Y=1,Z=1)
    Wave=(bEnabled=True,WaveFrequency=(X=95,Y=95,Z=95),WaveAmplitude=(X=0,Y=0.3,Z=0.3),WaveOffset=(X=0,Y=1,Z=2),WaveNoise=(X=0,Y=0.25,Z=0.35),WaveMinimum=0,WaveMaximum=0.35,WaveEase=SineEaseOut)
    WaveCurve=(EditorCurveData=(Keys=((),(Time=1,Value=1)),PreInfinityExtrap=RCCE_Constant,PostInfinityExtrap=RCCE_Constant,DefaultValue=340282346638528859811704183484516925440))
    bDrawDebug=True
   )
   > &RigUnit_ChainHarmonics(
    ChainRoot=dyn_squiggleL_01_Base
    DrawWorldOffset=(Rotation=(X=0,Y=0,Z=0,W=1),Translation=(X=0,Y=0,Z=0),Scale3D=(X=1,Y=1,Z=1))
    Pendulum=(bEnabled=True,PendulumStiffness=2,PendulumGravity=(X=0,Y=0,Z=0),PendulumBlend=0.75,PendulumDrag=0.98,PendulumMinimum=0,PendulumMaximum=0.1,PendulumEase=Linear,UnwindAxis=(X=0,Y=1,Z=0),UnwindMinimum=0.2,UnwindMaximum=0.05)
    Reach=(bEnabled=False,ReachTarget=(X=0,Y=0,Z=0),ReachAxis=(X=1,Y=0,Z=0),ReachMinimum=0,ReachMaximum=0,ReachEase=Linear)
    Speed=(X=1,Y=1,Z=1)
    Wave=(bEnabled=True,WaveFrequency=(X=93,Y=93,Z=9390),WaveAmplitude=(X=0,Y=0.35,Z=0.4),WaveOffset=(X=0,Y=1,Z=2),WaveNoise=(X=0,Y=0,Z=0),WaveMinimum=0,WaveMaximum=0.4,WaveEase=SineEaseOut)
    WaveCurve=(EditorCurveData=(Keys=((),(Time=1,Value=1)),PreInfinityExtrap=RCCE_Constant,PostInfinityExtrap=RCCE_Constant,DefaultValue=340282346638528859811704183484516925440))
    bDrawDebug=True
   )
   > &RigUnit_ChainHarmonics(
    ChainRoot=dyn_squiggleM_01_Base
    DrawWorldOffset=(Rotation=(X=0,Y=0,Z=0,W=1),Translation=(X=0,Y=0,Z=0),Scale3D=(X=1,Y=1,Z=1))
    Pendulum=(bEnabled=True,PendulumStiffness=2,PendulumGravity=(X=0,Y=0,Z=0),PendulumBlend=0.75,PendulumDrag=0.98,PendulumMinimum=0,PendulumMaximum=0.1,PendulumEase=Linear,UnwindAxis=(X=0,Y=1,Z=0),UnwindMinimum=0.2,UnwindMaximum=0.05)
    Reach=(bEnabled=False,ReachTarget=(X=0,Y=0,Z=0),ReachAxis=(X=1,Y=0,Z=0),ReachMinimum=0,ReachMaximum=0,ReachEase=Linear)
    Speed=(X=1,Y=1,Z=1)
    Wave=(bEnabled=True,WaveFrequency=(X=108,Y=108,Z=108),WaveAmplitude=(X=0,Y=0.5,Z=0.5),WaveOffset=(X=0,Y=1,Z=2),WaveNoise=(X=0,Y=0,Z=0),WaveMinimum=0,WaveMaximum=0.4,WaveEase=SineEaseOut)
    WaveCurve=(EditorCurveData=(Keys=((),(Time=1,Value=1)),PreInfinityExtrap=RCCE_Constant,PostInfinityExtrap=RCCE_Constant,DefaultValue=340282346638528859811704183484516925440))
    bDrawDebug=True
   )
   > &RigUnit_ChainHarmonics(
    ChainRoot=dyn_squiggleN_01_Base
    DrawWorldOffset=(Rotation=(X=0,Y=0,Z=0,W=1),Translation=(X=0,Y=0,Z=0),Scale3D=(X=1,Y=1,Z=1))
    Pendulum=(bEnabled=True,PendulumStiffness=2,PendulumGravity=(X=0,Y=0,Z=0),PendulumBlend=0.75,PendulumDrag=0.98,PendulumMinimum=0,PendulumMaximum=0.1,PendulumEase=Linear,UnwindAxis=(X=0,Y=1,Z=0),UnwindMinimum=0.2,UnwindMaximum=0.05)
    Reach=(bEnabled=False,ReachTarget=(X=0,Y=0,Z=0),ReachAxis=(X=1,Y=0,Z=0),ReachMinimum=0,ReachMaximum=0,ReachEase=Linear)
    Speed=(X=1,Y=1,Z=1)
    Wave=(bEnabled=True,WaveFrequency=(X=102,Y=102,Z=102),WaveAmplitude=(X=0,Y=0.15,Z=0.2),WaveOffset=(X=0,Y=1,Z=2),WaveNoise=(X=0,Y=0,Z=0),WaveMinimum=0,WaveMaximum=0.4,WaveEase=SineEaseOut)
    WaveCurve=(EditorCurveData=(Keys=((),(Time=1,Value=1)),PreInfinityExtrap=RCCE_Constant,PostInfinityExtrap=RCCE_Constant,DefaultValue=340282346638528859811704183484516925440))
    bDrawDebug=True
   )
   > &RigUnit_ChainHarmonics(
    ChainRoot=dyn_squiggleP_01_Base
    DrawWorldOffset=(Rotation=(X=0,Y=0,Z=0,W=1),Translation=(X=0,Y=0,Z=0),Scale3D=(X=1,Y=1,Z=1))
    Pendulum=(bEnabled=True,PendulumStiffness=2,PendulumGravity=(X=0,Y=0,Z=0),PendulumBlend=0.75,PendulumDrag=0.98,PendulumMinimum=0,PendulumMaximum=0.1,PendulumEase=Linear,UnwindAxis=(X=0,Y=1,Z=0),UnwindMinimum=0.2,UnwindMaximum=0.05)
    Reach=(bEnabled=False,ReachTarget=(X=0,Y=0,Z=0),ReachAxis=(X=1,Y=0,Z=0),ReachMinimum=0,ReachMaximum=0,ReachEase=Linear)
    Speed=(X=1,Y=1,Z=1)
    Wave=(bEnabled=True,WaveFrequency=(X=90,Y=80,Z=80),WaveAmplitude=(X=0,Y=0.35,Z=0.35),WaveOffset=(X=0,Y=1,Z=2),WaveNoise=(X=0,Y=0,Z=0),WaveMinimum=0,WaveMaximum=0.4,WaveEase=SineEaseOut)
    WaveCurve=(EditorCurveData=(Keys=((),(Time=1,Value=1)),PreInfinityExtrap=RCCE_Constant,PostInfinityExtrap=RCCE_Constant,DefaultValue=340282346638528859811704183484516925440))
    bDrawDebug=True
   )
   > &RigUnit_ChainHarmonics(
    ChainRoot=dyn_squiggleR_01_Base
    DrawWorldOffset=(Rotation=(X=0,Y=0,Z=0,W=1),Translation=(X=0,Y=0,Z=0),Scale3D=(X=1,Y=1,Z=1))
    Pendulum=(bEnabled=True,PendulumStiffness=2,PendulumGravity=(X=0,Y=0,Z=0),PendulumBlend=0.75,PendulumDrag=0.98,PendulumMinimum=0,PendulumMaximum=0.1,PendulumEase=Linear,UnwindAxis=(X=0,Y=1,Z=0),UnwindMinimum=0.2,UnwindMaximum=0.05)
    Reach=(bEnabled=False,ReachTarget=(X=0,Y=0,Z=0),ReachAxis=(X=1,Y=0,Z=0),ReachMinimum=0,ReachMaximum=0,ReachEase=Linear)
    Speed=(X=1,Y=1,Z=1)
    Wave=(bEnabled=True,WaveFrequency=(X=86,Y=86,Z=86),WaveAmplitude=(X=0,Y=0.35,Z=0.35),WaveOffset=(X=0,Y=1,Z=2),WaveNoise=(X=0,Y=0,Z=0),WaveMinimum=0,WaveMaximum=0.4,WaveEase=SineEaseOut)
    WaveCurve=(EditorCurveData=(Keys=((),(Time=1,Value=1)),PreInfinityExtrap=RCCE_Constant,PostInfinityExtrap=RCCE_Constant,DefaultValue=340282346638528859811704183484516925440))
    bDrawDebug=True
   )
   > &RigUnit_ChainHarmonics(
    ChainRoot=dyn_squiggleS_01_Base
    DrawWorldOffset=(Rotation=(X=0,Y=0,Z=0,W=1),Translation=(X=0,Y=0,Z=0),Scale3D=(X=1,Y=1,Z=1))
    Pendulum=(bEnabled=True,PendulumStiffness=2,PendulumGravity=(X=0,Y=0,Z=0),PendulumBlend=0.75,PendulumDrag=0.98,PendulumMinimum=0,PendulumMaximum=0.1,PendulumEase=Linear,UnwindAxis=(X=0,Y=1,Z=0),UnwindMinimum=0.2,UnwindMaximum=0.05)
    Reach=(bEnabled=False,ReachTarget=(X=0,Y=0,Z=0),ReachAxis=(X=1,Y=0,Z=0),ReachMinimum=0,ReachMaximum=0,ReachEase=Linear)
    Speed=(X=1,Y=1,Z=1)
    Wave=(bEnabled=True,WaveFrequency=(X=108,Y=108,Z=108),WaveAmplitude=(X=0,Y=0.5,Z=0.5),WaveOffset=(X=0,Y=1,Z=2),WaveNoise=(X=0,Y=0,Z=0),WaveMinimum=0,WaveMaximum=0.4,WaveEase=SineEaseOut)
    WaveCurve=(EditorCurveData=(Keys=((),(Time=1,Value=1)),PreInfinityExtrap=RCCE_Constant,PostInfinityExtrap=RCCE_Constant,DefaultValue=340282346638528859811704183484516925440))
    bDrawDebug=True
   )
   > &RigUnit_ChainHarmonics(
    ChainRoot=dyn_squiggleT_01_Base
    DrawWorldOffset=(Rotation=(X=0,Y=0,Z=0,W=1),Translation=(X=0,Y=0,Z=0),Scale3D=(X=1,Y=1,Z=1))
    Pendulum=(bEnabled=True,PendulumStiffness=2,PendulumGravity=(X=0,Y=0,Z=0),PendulumBlend=0.75,PendulumDrag=0.98,PendulumMinimum=0,PendulumMaximum=0.1,PendulumEase=Linear,UnwindAxis=(X=0,Y=1,Z=0),UnwindMinimum=0.2,UnwindMaximum=0.05)
    Reach=(bEnabled=False,ReachTarget=(X=0,Y=0,Z=0),ReachAxis=(X=1,Y=0,Z=0),ReachMinimum=0,ReachMaximum=0,ReachEase=Linear)
    Speed=(X=1,Y=1,Z=1)
    Wave=(bEnabled=True,WaveFrequency=(X=102,Y=102,Z=102),WaveAmplitude=(X=0,Y=0.5,Z=0.5),WaveOffset=(X=0,Y=1,Z=2),WaveNoise=(X=0,Y=0,Z=0),WaveMinimum=0,WaveMaximum=0.4,WaveEase=SineEaseOut)
    WaveCurve=(EditorCurveData=(Keys=((),(Time=1,Value=1)),PreInfinityExtrap=RCCE_Constant,PostInfinityExtrap=RCCE_Constant,DefaultValue=340282346638528859811704183484516925440))
    bDrawDebug=True
   )
   /Script/ControlRig.RigUnit_TransformConstraint(
    BaseBone=None
    BaseTransform=(Rotation=(X=0,Y=0,Z=0,W=1),Translation=(X=0,Y=0,Z=0),Scale3D=(X=1,Y=1,Z=1))
    BaseTransformSpace=GlobalSpace
    Bone=dyn_centerClav_r
    ExecutePin=()
    Targets=((Transform=(Rotation=(X=0,Y=0,Z=0,W=1),Translation=(X=0,Y=0,Z=0),Scale3D=(X=1,Y=1,Z=1)),Weight=1,bMaintainOffset=True,Filter=(TranslationFilter=(bX=True,bY=True,bZ=True),RotationFilter=(bX=True,bY=True,bZ=True),ScaleFilter=(bX=True,bY=True,bZ=True))))
    bUseInitialTransforms=True
    Targets.0.Transform=&RigUnit_GetBoneTransform(Bone=clavicle_r).Transform
   ) {is=struct RigVMUnitNode}
   > &RigUnit_TransformConstraint(
    BaseBone=None
    BaseTransform=(Rotation=(X=0,Y=0,Z=0,W=1),Translation=(X=0,Y=0,Z=0),Scale3D=(X=1,Y=1,Z=1))
    BaseTransformSpace=GlobalSpace
    Bone=dyn_centerClav_l
    Targets=((Transform=(Rotation=(X=0,Y=0,Z=0,W=1),Translation=(X=0,Y=0,Z=0),Scale3D=(X=1,Y=1,Z=1)),Weight=1,bMaintainOffset=True,Filter=(TranslationFilter=(bX=True,bY=True,bZ=True),RotationFilter=(bX=True,bY=True,bZ=True),ScaleFilter=(bX=True,bY=True,bZ=True))))
    bUseInitialTransforms=True
    Targets.0.Transform=&RigUnit_GetBoneTransform(Bone=clavicle_l).Transform
   )
   n4 = &RigUnit_GetBoneTransform(Bone=dyn_centerClavEND_r)
   > /Script/ControlRig.RigUnit_SetBoneTranslation(
    Bone=dyn_shoulderFloat_r
    Space=GlobalSpace
    Translation=(X=0,Y=0,Z=0)
    Weight=1
    bPropagateToChildren=True
    Translation.X=&RigUnit_GetBoneTransform(Bone=dyn_shoulderFloat_r).Transform.Translation.X
    Translation.Y=@n4.Transform.Translation.Y
    Translation.Z=@n4.Transform.Translation.Z
   ) {is=struct RigVMUnitNode}
   n5 = &RigUnit_GetBoneTransform(Bone=dyn_centerClavEND_l)
   > &RigUnit_SetBoneTranslation(
    Bone=dyn_shoulderFloat_l
    Space=GlobalSpace
    Translation=(X=0,Y=0,Z=0)
    Weight=1
    bPropagateToChildren=True
    Translation.X=&RigUnit_GetBoneTransform(Bone=dyn_shoulderFloat_l).Transform.Translation.X
    Translation.Y=@n5.Transform.Translation.Y
    Translation.Z=@n5.Transform.Translation.Z
   )
   &RigUnit_TransformConstraint(
    BaseBone=None
    BaseTransform=(Rotation=(X=0,Y=0,Z=0,W=1),Translation=(X=0,Y=0,Z=0),Scale3D=(X=1,Y=1,Z=1))
    BaseTransformSpace=GlobalSpace
    Bone=upperarm_l
    ExecutePin=()
    Targets=((Transform=(Rotation=(X=0,Y=0,Z=0,W=1),Translation=(X=0,Y=0,Z=0),Scale3D=(X=1,Y=1,Z=1)),Weight=1,bMaintainOffset=True,Filter=(TranslationFilter=(bX=True,bY=True,bZ=True),RotationFilter=(bX=True,bY=True,bZ=True),ScaleFilter=(bX=True,bY=True,bZ=True))))
    bUseInitialTransforms=True
    Targets.0.Transform.Translation=&RigUnit_GetBoneTransform(Bone=dyn_shoulderFloat_l).Transform.Translation
    Targets.0.Transform.Rotation=&RigUnit_GetBoneTransform(Bone=upperarm_l).Transform.Rotation
   )
   &RigUnit_TransformConstraint(
    BaseBone=None
    BaseTransform=(Rotation=(X=0,Y=0,Z=0,W=1),Translation=(X=0,Y=0,Z=0),Scale3D=(X=1,Y=1,Z=1))
    BaseTransformSpace=GlobalSpace
    Bone=upperarm_r
    ExecutePin=()
    Targets=((Transform=(Rotation=(X=0,Y=0,Z=0,W=1),Translation=(X=0,Y=0,Z=0),Scale3D=(X=1,Y=1,Z=1)),Weight=1,bMaintainOffset=True,Filter=(TranslationFilter=(bX=True,bY=True,bZ=True),RotationFilter=(bX=True,bY=True,bZ=True),ScaleFilter=(bX=True,bY=True,bZ=True))))
    bUseInitialTransforms=True
    Targets.0.Transform.Translation=&RigUnit_GetBoneTransform(Bone=dyn_shoulderFloat_r).Transform.Translation
    Targets.0.Transform.Rotation=&RigUnit_GetBoneTransform(Bone=upperarm_r).Transform.Rotation
   )
  ]}
]
hierarchy [
 root {loc=<0 0 0>, rot=<0 0 0 1>}
 attach {parent=root, loc=<0 0 0>, rot=<0 0 0 1>}
 pelvis {parent=root, loc=<0 -0.7071 93.1191>, rot=<0.0127 0.707 -0.0127 -0.707>}
 spine_01 {parent=pelvis, loc=<6.7494 -0.6388 0>, rot=<0 0 0.034 -0.9994>}
 spine_02 {parent=spine_01, loc=<7.2518 -0.035 0>, rot=<0 0 0.0238 -0.9997>}
 spine_03 {parent=spine_02, loc=<6.0104 0 0>, rot=<0 0 0.0149 -0.9999>}
 spine_04 {parent=spine_03, loc=<8.1024 0 0>, rot=<0 0 -0.0132 -0.9999>}
 spine_05 {parent=spine_04, loc=<10.0146 0 0>, rot=<0 0 -0.0015 -1>}
 clavicle_l {parent=spine_05, loc=<4.3575 -0.1466 -2.3223>, rot=<0.0465 -0.7289 -0.009 -0.6829>}
 upperarm_l {parent=clavicle_l, loc=<15.5111 0 0>, rot=<0.0008 0.0336 -0.0252 -0.9991>}
 lowerarm_l {parent=upperarm_l, loc=<26.6514 0 0>, rot=<0 0 0 1>}
 hand_l {parent=lowerarm_l, loc=<22.3491 0 0>, rot=<-0.7071 0 0 -0.7071>}
 index_metacarpal_l {parent=hand_l, loc=<1.6535 0.0183 -2.96>, rot=<0.0158 -0.0124 -0.0173 -0.9996>}
 index_01_l {parent=index_metacarpal_l, loc=<8.3008 0.0357 -0.0524>, rot=<-0.0044 0.0485 0.0924 -0.9945>}
 index_02_l {parent=index_01_l, loc=<3.3587 -0.3004 0.0452>, rot=<-0.002 0.0027 0.3225 -0.9466>}
 index_03_l {parent=index_02_l, loc=<2.5578 -0.123 -0.0252>, rot=<-0.002 0.0067 0.2558 -0.9667>}
 middle_metacarpal_l {parent=hand_l, loc=<2.1856 -0.1725 -0.5984>, rot=<0.006 0.0168 -0.0329 -0.9993>}
 middle_01_l {parent=middle_metacarpal_l, loc=<7.3275 -0.2447 0.0129>, rot=<-0.0208 0.0042 0.0858 -0.9961>}
 middle_02_l {parent=middle_01_l, loc=<4.0123 0.1401 0.1175>, rot=<-0.0218 0.0258 0.3512 -0.9357>}
 middle_03_l {parent=middle_02_l, loc=<2.6077 0.0534 0.1373>, rot=<0.0082 0.0188 0.1708 -0.9851>}
 pinky_metacarpal_l {parent=hand_l, loc=<1.7616 -0.9753 3.3814>, rot=<-0.0397 0.0814 -0.069 -0.9935>}
 pinky_01_l {parent=pinky_metacarpal_l, loc=<7.1709 -0.4831 -0.1094>, rot=<-0.0654 -0.0035 0.2242 -0.9723>}
 pinky_02_l {parent=pinky_01_l, loc=<2.8274 0.0893 -0.1043>, rot=<0.0037 0.0011 0.2854 -0.9584>}
 pinky_03_l {parent=pinky_02_l, loc=<2.2047 0.0307 -0.0078>, rot=<-0.0002 0.0098 0.1445 -0.9895>}
 ring_metacarpal_l {parent=hand_l, loc=<2.2074 -0.3625 1.3638>, rot=<-0.0284 0.0722 -0.0458 -0.9959>}
 ring_01_l {parent=ring_metacarpal_l, loc=<7.0295 -0.6197 -0.1446>, rot=<-0.0489 -0.0252 0.1892 -0.9804>}
 ring_02_l {parent=ring_01_l, loc=<3.6174 0.4646 -0.1027>, rot=<0 0 0.2806 -0.9598>}
 ring_03_l {parent=ring_02_l, loc=<2.4152 0.1141 0.0199>, rot=<0 0 0.1189 -0.9929>}
 thumb_01_l {parent=hand_l, loc=<1.3118 -0.3191 -2.459>, rot=<0.6546 -0.4042 0.0664 -0.6354>}
 thumb_02_l {parent=thumb_01_l, loc=<5.9336 -0.0448 0.0401>, rot=<-0.0037 0.0213 0.298 -0.9543>}
 thumb_03_l {parent=thumb_02_l, loc=<2.2223 0.3682 -0.0907>, rot=<0.0118 -0.013 0.0066 -0.9998>}
 weapon_l {parent=hand_l, loc=<-0.1974 -3.4111 0.3811>, rot=<0 0 0 1>}
 lowerarm_twist_01_l {parent=lowerarm_l, loc=<17.8549 0 0>, rot=<0 0 0 1>}
 lowerarm_twist_02_l {parent=lowerarm_l, loc=<9.6199 0 0>, rot=<0 0 0 1>}
 dyn_forearmWobble_Base_l {parent=lowerarm_l, loc=<-6.7453 -0.9485 -17.6959>, rot=<-0.0379 -0.3574 -0.2294 -0.9046>}
 dyn_forearmWobble_l {parent=dyn_forearmWobble_Base_l, loc=<0 0 0>, rot=<0 0 0 1>}
 dyn_handToElbow_l {parent=lowerarm_l, loc=<7.5427 8.9327 -31.7571>, rot=<0.41 -0.6955 0.5796 0.1106>}
 upperarm_twist_01_l {parent=upperarm_l, loc=<0.0041 0 0>, rot=<0 0 0 1>}
 upperarm_twist_02_l {parent=upperarm_l, loc=<13.7941 0 0>, rot=<0 0 0 1>}
 dyn_shoulderToElbow_l {parent=upperarm_l, loc=<0.0031 -0.0001 -0.0021>, rot=<0.6548 -0.2668 -0.2353 -0.6668>}
 clavicle_r {parent=spine_05, loc=<4.3575 -0.1466 2.3223>, rot=<0.7289 0.0465 -0.6829 0.009>}
 upperarm_r {parent=clavicle_r, loc=<-15.511 0 0.0001>, rot=<0.0008 0.0336 -0.0252 -0.9991>}
 lowerarm_r {parent=upperarm_r, loc=<-26.6514 0 0>, rot=<0 0 0 1>}
 hand_r {parent=lowerarm_r, loc=<-22.3492 0 0>, rot=<-0.7071 0 0 -0.7071>}
 index_metacarpal_r {parent=hand_r, loc=<-1.6534 -0.0181 2.9601>, rot=<0.0158 -0.0124 -0.0173 -0.9996>}
 index_01_r {parent=index_metacarpal_r, loc=<-8.3008 -0.0357 0.0524>, rot=<-0.0044 0.0485 0.0924 -0.9945>}
 index_02_r {parent=index_01_r, loc=<-3.3589 0.3003 -0.0453>, rot=<-0.002 0.0027 0.3225 -0.9466>}
 index_03_r {parent=index_02_r, loc=<-2.5578 0.1231 0.0251>, rot=<-0.002 0.0067 0.2558 -0.9667>}
 middle_metacarpal_r {parent=hand_r, loc=<-2.1856 0.1726 0.5984>, rot=<0.006 0.0168 -0.0329 -0.9993>}
 middle_01_r {parent=middle_metacarpal_r, loc=<-7.3273 0.2447 -0.0129>, rot=<-0.0208 0.0042 0.0858 -0.9961>}
 middle_02_r {parent=middle_01_r, loc=<-4.0142 -0.1411 -0.07>, rot=<-0.0217 0.0259 0.3512 -0.9357>}
 middle_03_r {parent=middle_02_r, loc=<-2.6074 -0.0534 -0.1312>, rot=<0.0083 0.0188 0.1708 -0.9851>}
 pinky_metacarpal_r {parent=hand_r, loc=<-1.7617 0.9753 -3.3814>, rot=<-0.0397 0.0814 -0.069 -0.9935>}
 pinky_01_r {parent=pinky_metacarpal_r, loc=<-7.1707 0.4832 0.1094>, rot=<-0.0654 -0.0035 0.2242 -0.9723>}
 pinky_02_r {parent=pinky_01_r, loc=<-2.8274 -0.0892 0.1043>, rot=<0.0037 0.0011 0.2854 -0.9584>}
 pinky_03_r {parent=pinky_02_r, loc=<-2.2045 -0.031 0.0078>, rot=<-0.0002 0.0098 0.1445 -0.9895>}
 ring_metacarpal_r {parent=hand_r, loc=<-2.2073 0.3627 -1.3638>, rot=<-0.0284 0.0722 -0.0458 -0.9959>}
 ring_01_r {parent=ring_metacarpal_r, loc=<-7.0297 0.6197 0.1446>, rot=<-0.0489 -0.0252 0.1892 -0.9804>}
 ring_02_r {parent=ring_01_r, loc=<-3.6171 -0.4647 0.1026>, rot=<0 0 0.2806 -0.9598>}
 ring_03_r {parent=ring_02_r, loc=<-2.4153 -0.1142 -0.0199>, rot=<0 0 0.1189 -0.9929>}
 thumb_01_r {parent=hand_r, loc=<-1.3119 0.3192 2.459>, rot=<0.6546 -0.4042 0.0664 -0.6354>}
 thumb_02_r {parent=thumb_01_r, loc=<-5.9336 0.0447 -0.0401>, rot=<-0.0037 0.0213 0.298 -0.9543>}
 thumb_03_r {parent=thumb_02_r, loc=<-2.2223 -0.3682 0.0907>, rot=<0.0118 -0.013 0.0066 -0.9998>}
 weapon_r {parent=hand_r, loc=<0.1975 3.4112 -0.3811>, rot=<0 0 0 -1>}
 lowerarm_twist_01_r {parent=lowerarm_r, loc=<-17.8549 0 0>, rot=<0 0 0 1>}
 lowerarm_twist_02_r {parent=lowerarm_r, loc=<-9.6199 0 0>, rot=<0 0 0 1>}
 dyn_forearmWobble_Base_r {parent=lowerarm_r, loc=<6.7453 0.9485 17.6959>, rot=<-0.0379 -0.3574 -0.2294 -0.9046>}
 dyn_forearmWobble_r {parent=dyn_forearmWobble_Base_r, loc=<0 0 0>, rot=<0 0 0 1>}
 dyn_handToElbow_r {parent=lowerarm_r, loc=<-7.5427 -8.9327 31.7567>, rot=<-0.6955 -0.41 0.1106 -0.5796>}
 upperarm_twist_01_r {parent=upperarm_r, loc=<-0.0041 0 0>, rot=<0 0 0 1>}
 upperarm_twist_02_r {parent=upperarm_r, loc=<-13.7942 0 0>, rot=<0 0 0 1>}
 dyn_shoulderToElbow_r {parent=upperarm_r, loc=<-0.0032 0.0001 0.0027>, rot=<0.2668 0.6548 0.6668 -0.2353>}
 neck_01 {parent=spine_05, loc=<11.3138 0.1529 0>, rot=<0 0 -0.2003 -0.9797>}
 neck_02 {parent=neck_01, loc=<4.5627 0 0>, rot=<0 0 0.0504 -0.9987>}
 head {parent=neck_02, loc=<5.391 0 0>, rot=<0 0 0.115 -0.9934>}
 dyn_headBase {parent=head, loc=<-5.414 -0.0447 0>, rot=<0 0 0 -1>}
 dyn_squiggleA_01_Base {parent=dyn_headBase, loc=<18.6353 2.7234 -6.8429>, rot=<0.9451 0.0237 -0.1431 0.2929>}
 dyn_squiggleA_01 {parent=dyn_squiggleA_01_Base, loc=<0 0 0>, rot=<0 0 0 1>}
 dyn_squiggleA_02 {parent=dyn_squiggleA_01, loc=<7.8589 0 0>, rot=<-0.0328 -0.131 -0.1413 -0.9807>}
 dyn_squiggleA_03 {parent=dyn_squiggleA_02, loc=<9.2957 0 0>, rot=<-0.2995 0.0586 0.1321 -0.9431>}
 dyn_squiggleB_01_Base {parent=dyn_headBase, loc=<10.3387 -5.3694 -4.7771>, rot=<0.9407 -0.1178 -0.0421 -0.3152>}
 dyn_squiggleB_01 {parent=dyn_squiggleB_01_Base, loc=<0 0 0>, rot=<0 0 0 1>}
 dyn_squiggleB_02 {parent=dyn_squiggleB_01, loc=<8.1661 0 0>, rot=<0.0589 -0.0253 -0.148 -0.9869>}
 dyn_squiggleB_03 {parent=dyn_squiggleB_02, loc=<8.0616 0 0>, rot=<0.0016 -0.0007 0.1397 -0.9902>}
 dyn_squiggleC_01_Base {parent=dyn_headBase, loc=<12.435 -8.9149 -2.396>, rot=<0.9795 -0.2013 -0.0007 -0.0031>}
 dyn_squiggleC_01 {parent=dyn_squiggleC_01_Base, loc=<0 0 0>, rot=<0 0 0 -1>}
 dyn_squiggleC_02 {parent=dyn_squiggleC_01, loc=<8.1596 0 0>, rot=<-0.1039 0.0276 0.1523 -0.9825>}
 dyn_squiggleD_01_Base {parent=dyn_headBase, loc=<9.4408 -8.0399 0.8968>, rot=<0.9546 -0.2177 0.0469 0.1979>}
 dyn_squiggleD_01 {parent=dyn_squiggleD_01_Base, loc=<0 0 0>, rot=<0 0 0 -1>}
 dyn_squiggleD_02 {parent=dyn_squiggleD_01, loc=<11.5577 0 0>, rot=<-0.1815 0.0504 0.1814 -0.9652>}
 dyn_squiggleD_03 {parent=dyn_squiggleD_02, loc=<11.3888 0 0>, rot=<0.002 0.0927 -0.0898 -0.9916>}
 dyn_squiggleE_01_Base {parent=dyn_headBase, loc=<8.9642 -7.3785 4.6495>, rot=<0.879 -0.1141 0.0634 0.4586>}
 dyn_squiggleE_01 {parent=dyn_squiggleE_01_Base, loc=<0 0 0>, rot=<0 0 0 1>}
 dyn_squiggleE_02 {parent=dyn_squiggleE_01, loc=<8.8957 0 0>, rot=<0.0562 -0.0155 -0.0035 -0.9983>}
 dyn_squiggleE_03 {parent=dyn_squiggleE_02, loc=<8.541 0 0>, rot=<0.0812 0.025 0.172 -0.9814>}
 dyn_squiggleF_01_Base {parent=dyn_headBase, loc=<11.9754 -5.1814 7.3929>, rot=<-0.8186 0.1346 -0.0951 -0.5502>}
 dyn_squiggleF_01 {parent=dyn_squiggleF_01_Base, loc=<0 0 0>, rot=<0 0 0 -1>}
 dyn_squiggleF_02 {parent=dyn_squiggleF_01, loc=<13.2006 0 0>, rot=<0.0481 0.0342 0.1447 -0.9877>}
 dyn_squiggleF_03 {parent=dyn_squiggleF_02, loc=<6.5621 0 0>, rot=<0.1244 0.1404 0.159 -0.9693>}
 dyn_squiggleG_01_Base {parent=dyn_headBase, loc=<13.8803 -0.9443 7.4847>, rot=<-0.6497 0.1704 -0.1938 -0.715>}
 dyn_squiggleG_01 {parent=dyn_squiggleG_01_Base, loc=<0 0 0>, rot=<0 0 0 -1>}
 dyn_squiggleG_02 {parent=dyn_squiggleG_01, loc=<10.2479 0 0>, rot=<0.0651 0.0261 0.2303 -0.9706>}
 dyn_squiggleG_03 {parent=dyn_squiggleG_02, loc=<6.4581 0 0>, rot=<-0.0973 0.0009 0.1489 -0.9841>}
 dyn_squiggleH_01_Base {parent=dyn_headBase, loc=<17.8465 3.4097 4.5514>, rot=<-0.6495 0.1493 -0.173 -0.7252>}
 dyn_squiggleH_01 {parent=dyn_squiggleH_01_Base, loc=<0 0 0>, rot=<0 0 0 -1>}
 dyn_squiggleH_02 {parent=dyn_squiggleH_01, loc=<8.4848 0 0>, rot=<0.0525 -0.031 -0.0741 -0.9954>}
 dyn_squiggleH_03 {parent=dyn_squiggleH_02, loc=<5.9779 0 0>, rot=<-0.0161 0.0625 0.3382 -0.9389>}
 dyn_squiggleK_01_Base {parent=dyn_headBase, loc=<14.6273 -3.2875 -3.8563>, rot=<0.8712 -0.1354 -0.0763 -0.4657>}
 dyn_squiggleK_01 {parent=dyn_squiggleK_01_Base, loc=<0 0 0>, rot=<0 0 0 -1>}
 dyn_squiggleK_02 {parent=dyn_squiggleK_01, loc=<12.4989 0 0>, rot=<-0.1617 0.0614 -0.0457 -0.9839>}
 dyn_squiggleK_03 {parent=dyn_squiggleK_02, loc=<9.0056 0 0>, rot=<-0.0579 0.0397 -0.1896 -0.9793>}
 dyn_squiggleL_01_Base {parent=dyn_headBase, loc=<18.7053 3.2731 -3.9053>, rot=<0.9766 -0.2022 -0.0155 -0.072>}
 dyn_squiggleL_01 {parent=dyn_squiggleL_01_Base, loc=<0 0 0>, rot=<0 0 0 -1>}
 dyn_squiggleL_02 {parent=dyn_squiggleL_01, loc=<10.4186 0 0>, rot=<0.1433 -0.1185 0.1427 -0.9722>}
 dyn_squiggleL_03 {parent=dyn_squiggleL_02, loc=<10.6574 0 0>, rot=<0.04 0.0671 -0.0888 -0.993>}
 dyn_squiggleM_01_Base {parent=dyn_headBase, loc=<19.144 3.7086 0.4851>, rot=<0.0819 -0.0524 -0.0749 -0.9924>}
 dyn_squiggleM_01 {parent=dyn_squiggleM_01_Base, loc=<0 0 0>, rot=<0 0 0 1>}
 dyn_squiggleM_02 {parent=dyn_squiggleM_01, loc=<7.1074 0 0>, rot=<0.0302 -0.1324 0.0835 -0.9872>}
 dyn_squiggleM_03 {parent=dyn_squiggleM_02, loc=<5.7465 0 0>, rot=<-0.0402 0.1806 -0.1096 -0.9766>}
 dyn_squiggleN_01_Base {parent=dyn_headBase, loc=<22.9026 -0.0288 0.8135>, rot=<0.9916 -0.0796 -0.009 -0.1013>}
 dyn_squiggleN_01 {parent=dyn_squiggleN_01_Base, loc=<0 0 0>, rot=<0 0 0 1>}
 dyn_squiggleN_02 {parent=dyn_squiggleN_01, loc=<8.9668 0 0>, rot=<-0.0323 -0.1436 0.0803 -0.9859>}
 dyn_squiggleP_01_Base {parent=dyn_headBase, loc=<12.6909 -4.4883 4.0525>, rot=<0.9776 -0.0161 0.1096 -0.1791>}
 dyn_squiggleP_01 {parent=dyn_squiggleP_01_Base, loc=<0 0 0>, rot=<0 0 0 -1>}
 dyn_squiggleP_02 {parent=dyn_squiggleP_01, loc=<9.3378 0 0>, rot=<0.1955 0.0627 0.0536 -0.9772>}
 dyn_squiggleP_03 {parent=dyn_squiggleP_02, loc=<6.7058 0 0>, rot=<-0.0629 0.0583 -0.1336 -0.9873>}
 dyn_squiggleR_01_Base {parent=dyn_headBase, loc=<17.3621 0.4842 3.4786>, rot=<0.8266 0.0208 0.1845 -0.5313>}
 dyn_squiggleR_01 {parent=dyn_squiggleR_01_Base, loc=<0 0 0>, rot=<0 0 0 -1>}
 dyn_squiggleR_02 {parent=dyn_squiggleR_01, loc=<11.4279 0 0>, rot=<-0.0022 0.0953 0.0593 -0.9937>}
 dyn_squiggleR_03 {parent=dyn_squiggleR_02, loc=<7.9361 0 0>, rot=<0.0296 0.1273 -0.2926 -0.9473>}
 dyn_squiggleS_01_Base {parent=dyn_headBase, loc=<15.505 1.6526 7.6228>, rot=<-0.4898 0.3087 -0.4404 -0.6862>}
 dyn_squiggleS_01 {parent=dyn_squiggleS_01_Base, loc=<0 0 0>, rot=<0 0 0 1>}
 dyn_squiggleS_02 {parent=dyn_squiggleS_01, loc=<4.9536 0 0>, rot=<-0.0172 0.0228 0.2067 -0.978>}
 dyn_squiggleS_03 {parent=dyn_squiggleS_02, loc=<4.4258 0 0>, rot=<-0.3606 0.1991 0.1915 -0.8909>}
 dyn_squiggleJ_01_Base {parent=dyn_headBase, loc=<16.5988 -1.8071 -8.239>, rot=<0.7481 -0.2179 -0.1802 -0.6003>}
 dyn_squiggleJ_01 {parent=dyn_squiggleJ_01_Base, loc=<0 0 0>, rot=<0 0 0 -1>}
 dyn_squiggleJ_02 {parent=dyn_squiggleJ_01, loc=<5.3363 0 0>, rot=<-0.0154 -0.0412 0.1205 -0.9917>}
 dyn_squiggleJ_03 {parent=dyn_squiggleJ_02, loc=<5.9323 0 0>, rot=<-0.0873 0.0955 -0.1579 -0.9789>}
 dyn_squiggleT_01_Base {parent=dyn_headBase, loc=<16.0465 -4.0242 -0.1804>, rot=<0.0083 -0.0402 0.1944 -0.9801>}
 dyn_squiggleT_01 {parent=dyn_squiggleT_01_Base, loc=<0 0 0>, rot=<0 0 0 -1>}
 dyn_squiggleT_02 {parent=dyn_squiggleT_01, loc=<8.6967 0 0>, rot=<0.0229 -0.0642 -0.0569 -0.9961>}
 dyn_squiggleT_03 {parent=dyn_squiggleT_02, loc=<7.3091 0 0>, rot=<-0.0389 0.1308 0.0035 -0.9906>}
 faceAttach {parent=head, loc=<2.2332 4.3699 0>, rot=<0.0082 -0.7071 -0.0082 -0.7071>, scale=<0.89 0.89 0.89>}
 C_brow_mid {parent=faceAttach, loc=<-0.0019 9.1997 5.3811>, rot=<0 0 0 -1>}
 L_brow_mid {parent=faceAttach, loc=<3.2477 8.9096 5.9841>, rot=<-0.0001 -0.0042 0.2553 -0.9669>}
 L_brow_outer {parent=faceAttach, loc=<5.7983 6.59 5.234>, rot=<-0.0286 -0.0399 0.4491 -0.8921>}
 L_cheek_inner {parent=faceAttach, loc=<11.0468 2.7779 -0.8804>, rot=<-0.0786 -0.1491 0.414 -0.8945>}
 C_lip_upper_mid {parent=faceAttach, loc=<-0.0076 8.4603 -1.9012>, rot=<0.0454 -0.0005 0.0003 -0.999>}
 L_lip_upper_outer {parent=faceAttach, loc=<4.874 8.046 -1.8698>, rot=<0.039 -0.0294 0.2765 -0.9598>}
 L_lip_corner {parent=faceAttach, loc=<8.5093 6.0924 -2.6484>, rot=<-0.0085 0.0123 -0.0141 -0.9998>}
 C_nose_bridge {parent=faceAttach, loc=<-0.0198 8.8234 1.5094>, rot=<-0.0488 0 0 -0.9988>}
 R_brow_mid {parent=faceAttach, loc=<-3.2428 8.901 5.9841>, rot=<-0.0001 0.0042 -0.2553 -0.9669>}
 R_brow_outer {parent=faceAttach, loc=<-5.7792 6.5767 5.235>, rot=<-0.0286 0.04 -0.4491 -0.8921>}
 R_cheek_inner {parent=faceAttach, loc=<-11.4683 3.4501 -1.9386>, rot=<-0.0786 0.1491 -0.414 -0.8945>}
 R_lip_corner {parent=faceAttach, loc=<-7.3617 6.3347 -2.4683>, rot=<-0.0056 -0.0093 0.0144 -0.9998>}
 C_jaw {parent=faceAttach, loc=<-0.0151 -0.1812 -1.6057>, rot=<0 0 0 1>}
 L_lip_lower_outer {parent=C_jaw, loc=<4.3051 7.9611 -2.0026>, rot=<0.0769 -0.0136 0.2146 -0.9736>}
 C_lip_lower_mid {parent=C_jaw, loc=<-0.0226 8.2933 -2.0679>, rot=<-0.0084 -0.0148 0.0133 -0.9998>}
 R_lip_lower_outer {parent=C_jaw, loc=<-4.2839 7.9618 -2.0022>, rot=<0.0766 0.0149 -0.2134 -0.9739>}
 teeth_lower {parent=C_jaw, loc=<0.0151 6.8354 -1.9816>, rot=<0 0 0 1>}
 tongue {parent=C_jaw, loc=<0.0158 3.4219 -0.4018>, rot=<-0.1568 0 0 -0.9876>}
 R_eye {parent=faceAttach, loc=<-3.1514 6.6362 4.3311>, rot=<0 0 0.9998 -0.0187>}
 R_eye_lid_upper_mid {parent=R_eye, loc=<-0.0396 -1.1712 0.2636>, rot=<0.0523 -0.0009 0.0167 -0.9985>}
 R_eye_lid_lower_mid {parent=R_eye, loc=<-0.0085 -0.9422 -0.7046>, rot=<0.0329 -0.0005 0.0167 -0.9993>}
 L_eye {parent=faceAttach, loc=<3.1514 6.6362 4.3311>, rot=<0 0 0.9998 0.0187>}
 L_eye_lid_upper_mid {parent=L_eye, loc=<0.0251 -1.1623 0.2388>, rot=<0.0523 0 0 -0.9986>}
 L_eye_lid_lower_mid {parent=L_eye, loc=<0.0223 -0.923 -0.7488>, rot=<0.0329 0 0 -0.9995>}
 teeth_upper {parent=faceAttach, loc=<0 7.3065 -1.2818>, rot=<0 0 0 1>}
 R_lip_upper_outer {parent=faceAttach, loc=<-4.8003 7.9118 -1.855>, rot=<0.0389 0.0296 -0.2766 -0.9597>}
 dyn_centerClav_r {parent=spine_05, loc=<5.4536 2.6544 0>, rot=<0.7447 0.0937 -0.6592 -0.0462>}
 dyn_centerClavEND_r {parent=dyn_centerClav_r, loc=<-14.6157 0 0>, rot=<0.007 0.031 -0.0711 -0.997>}
 dyn_centerClav_l {parent=spine_05, loc=<5.4538 2.6544 0>, rot=<0.0937 -0.7447 0.0462 -0.6592>}
 dyn_centerClavEND_l {parent=dyn_centerClav_l, loc=<14.6158 0 0>, rot=<0.007 0.031 -0.0711 -0.997>}
 dyn_shoulderFloat_r {parent=spine_05, loc=<3.797 -0.2748 14.2231>, rot=<0.7447 0.0937 -0.6592 -0.0462>}
 dyn_shoulderFloat_l {parent=spine_05, loc=<3.7972 -0.2747 -14.2231>, rot=<0.0937 -0.7447 0.0462 -0.6592>}
 dyn_bandolierFront_01 {parent=spine_05, loc=<3.6147 11.416 -9.9417>, rot=<-0.0596 -0.8923 0.2321 0.3825>}
 dyn_bandolierFront_02 {parent=dyn_bandolierFront_01, loc=<8.7117 0 0>, rot=<0.2286 -0.0283 0.2654 -0.9362>}
 dyn_bandolierFront_03 {parent=dyn_bandolierFront_02, loc=<11.7595 2.3375 0.7573>, rot=<0.0651 -0.0344 0.0863 -0.9935>}
 dyn_chestPouch_01_Base {parent=spine_05, loc=<3.2977 0.9214 -9.4077>, rot=<0.0417 -0.9226 0.0041 0.3835>}
 dyn_chestPouch_01 {parent=dyn_chestPouch_01_Base, loc=<0 0 0>, rot=<0 0 0 -1>}
 dyn_chestPouch_02 {parent=dyn_chestPouch_01, loc=<29.6301 0 0>, rot=<-0.7255 0.1386 -0.1868 -0.6477>}
 dyn_bandolierSideBack {parent=dyn_chestPouch_02, loc=<-0.8216 4.0099 9.4678>, rot=<0.7217 -0.1218 0.155 -0.6636>}
 dyn_bandolierSideFront {parent=dyn_chestPouch_02, loc=<0.6833 -3.8768 -8.9985>, rot=<0.802 -0.0962 0.3297 -0.4887>}
 dyn_chestPouchEND {parent=dyn_chestPouch_02, loc=<17.5509 0 0>, rot=<0 0 0 -1>}
 dyn_bandolierBack_01 {parent=spine_04, loc=<8.5122 -9.9747 -5.026>, rot=<0.0375 -0.9366 0.0184 0.3478>}
 dyn_bandolierBack_02 {parent=dyn_bandolierBack_01, loc=<12.1036 -0.0113 0.0411>, rot=<-0.0008 -0.0024 -0.0007 -1>}
 thigh_l {parent=pelvis, loc=<-2.5693 -0.0109 -7.7501>, rot=<0.0012 0.0572 0.0155 -0.9982>}
 calf_l {parent=thigh_l, loc=<-39.0484 0 0>, rot=<0.0006 -0.0161 -0.0391 -0.9991>}
 calf_twist_01_l {parent=calf_l, loc=<-35.8131 -0.0967 0.0397>, rot=<0.0237 0.0095 -0.0035 -0.9997>}
 calf_twist_02_l {parent=calf_l, loc=<-18.8064 -0.0967 0.0397>, rot=<0.0237 0.0095 -0.0035 -0.9997>}
 foot_l {parent=calf_l, loc=<-42.561 0.1077 0.4292>, rot=<-0.0031 -0.0411 0.0414 -0.9983>}
 ball_l {parent=foot_l, loc=<-6.1599 13.8264 0.6806>, rot=<0.0001 -0.0001 -0.7063 -0.7079>}
 dyn_footToKnee_l {parent=calf_l, loc=<-42.561 0.1077 0.4292>, rot=<0.9976 -0.0009 -0.0051 0.0684>}
 dyn_calfWobble_Base_l {parent=calf_l, loc=<0 0 0>, rot=<0 0 0 1>}
 dyn_calfWobble_l {parent=dyn_calfWobble_Base_l, loc=<0 0 0>, rot=<0 0 0 1>}
 thigh_twist_01_l {parent=thigh_l, loc=<-21.6303 0 0>, rot=<0.0013 0 0 -1>}
 dyn_hipToKnee_l {parent=thigh_l, loc=<0 0 0>, rot=<0 0.0734 0.9973 0>}
 dyn_calfMeasure_l {parent=thigh_l, loc=<-39.0484 0 0>, rot=<0.0006 -0.0161 -0.0391 -0.9991>}
 thigh_r {parent=pelvis, loc=<-2.5691 -0.0109 7.7501>, rot=<0.0572 -0.0012 0.9982 0.0155>}
 calf_r {parent=thigh_r, loc=<39.0487 0 0>, rot=<0.0006 -0.0161 -0.0391 -0.9991>}
 calf_twist_01_r {parent=calf_r, loc=<35.8132 0.0967 -0.0397>, rot=<0.0237 0.0095 -0.0035 -0.9997>}
 calf_twist_02_r {parent=calf_r, loc=<18.8064 0.0967 -0.0397>, rot=<0.0237 0.0095 -0.0035 -0.9997>}
 foot_r {parent=calf_r, loc=<42.5612 -0.1077 -0.4292>, rot=<-0.0031 -0.0411 0.0414 -0.9983>}
 ball_r {parent=foot_r, loc=<6.1598 -13.8264 -0.6806>, rot=<0.0001 -0.0001 -0.7063 -0.7079>}
 dyn_footToKnee_r {parent=calf_r, loc=<26.8001 -0.1077 -0.4292>, rot=<-0.0009 -0.9976 0.0684 0.0051>}
 dyn_calfWobble_Base_r {parent=calf_r, loc=<0 0 0>, rot=<0 0 0 1>}
 dyn_calfWobble_r {parent=dyn_calfWobble_Base_r, loc=<0 0 0>, rot=<0 0 0 1>}
 thigh_twist_01_r {parent=thigh_r, loc=<21.6303 0 0>, rot=<0.0013 0 0 -1>}
 dyn_hipToKnee_r {parent=thigh_r, loc=<0 0 0>, rot=<0.0734 0 0 -0.9973>}
 dyn_calfMeasure_r {parent=thigh_r, loc=<39.0487 0 0>, rot=<0.0006 -0.0161 -0.0391 -0.9991>}
 dyn_pelvisBounce {parent=pelvis, loc=<0 0 0>, rot=<0 0 0 -1>}
 ik_foot_root {parent=root, loc=<0 0 0>, rot=<0 0 0 1>}
 ik_foot_l {parent=ik_foot_root, loc=<15.2608 -4.4176 9.4433>, rot=<-0.0357 -0.7045 -0.0284 0.7082>}
 ik_foot_r {parent=ik_foot_root, loc=<-15.2609 -4.4176 9.4431>, rot=<-0.7082 -0.0284 0.7045 -0.0357>}
 ik_hand_root {parent=root, loc=<0 0 0>, rot=<0 0 0 1>}
 ik_hand_gun {parent=ik_hand_root, loc=<-51.9729 4.0884 102.6935>, rot=<-0.6084 -0.1793 -0.4613 0.6205>}
 ik_hand_l {parent=ik_hand_gun, loc=<53.0286 82.1782 35.2051>, rot=<-0.5102 -0.7906 -0.3387 0>}
 ik_hand_r {parent=ik_hand_gun, loc=<0 0 0>, rot=<0 0 0 -1>}
]
