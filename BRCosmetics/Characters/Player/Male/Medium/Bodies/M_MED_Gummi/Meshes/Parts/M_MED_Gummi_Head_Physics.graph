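# Physics asset for the head part of the medium male Gummi player cosmetic, giving its "squiggle" strands bodies for secondary motion.
# A head box links to a small dyn_headBase box, from which 17 capsule chains branch, each with a base and two or three segments.
# Head and chain-base joints rotate freely, while segment joints carry tight swing and twist limits, with wider swing on chain C.
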PhysicsAsset /BRCosmetics/Characters/Player/Male/Medium/Bodies/M_MED_Gummi/Meshes/Parts/M_MED_Gummi_Head_Physics
bodies {
 dyn_headBase [{shapeType=Box, extentX=1.01, extentY=1.01, extentZ=1.01, rotation=<rot 0 -90 0>}]
 dyn_squiggleA_01 [{shapeType=Capsule, length=7.9632, radius=1.9624, center=<3 0 0>, rotation=<rot 0 90 90>}]
 dyn_squiggleA_01_Base [{shapeType=Capsule, length=0.5339, radius=0.5, rotation=<rot 0 -90 0>}]
 dyn_squiggleA_02 [{shapeType=Capsule, length=9.8374, radius=1.9624, center=<3 0 0>, rotation=<rot 0 90 90>}]
 dyn_squiggleA_03 [{shapeType=Capsule, length=9.0573, radius=1.9624, center=<3 0 0>, rotation=<rot 0 90 90>}]
 dyn_squiggleB_01 [{shapeType=Capsule, length=9.0573, radius=1.9624, center=<3 0 0>, rotation=<rot 0 90 90>}]
 dyn_squiggleB_01_Base [{shapeType=Capsule, length=0.5339, radius=0.5, rotation=<rot 0 -90 0>}]
 dyn_squiggleB_02 [{shapeType=Capsule, length=9.8374, radius=1.9624, center=<3 0 0>, rotation=<rot 0 90 90>}]
 dyn_squiggleB_03 [{shapeType=Capsule, length=5.5624, radius=1.9624, center=<1.0843 0 0>, rotation=<rot 0 90 90>}]
 dyn_squiggleC_01 [{shapeType=Capsule, length=6.1908, radius=1.8994, center=<2.0568 0 0>, rotation=<rot 0 90 90>}]
 dyn_squiggleC_01_Base [{shapeType=Capsule, length=3.7993, radius=0.5, rotation=<rot 0 -90 0>}]
 dyn_squiggleC_02 [{shapeType=Capsule, length=6.8178, radius=1.9624, center=<2.1056 0 0>, rotation=<rot 0 90 90>}]
 dyn_squiggleD_01 [{shapeType=Capsule, length=9.665, radius=1.9624, center=<3 0 0>, rotation=<rot 0 90 90>}]
 dyn_squiggleD_01_Base [{shapeType=Capsule, length=0.5339, radius=0.5, rotation=<rot 0 -90 0>}]
 dyn_squiggleD_02 [{shapeType=Capsule, length=9.8374, radius=1.9624, center=<3 0 0>, rotation=<rot 0 90 90>}]
 dyn_squiggleD_03 [{shapeType=Capsule, length=9.0573, radius=1.9624, center=<3 0 0>, rotation=<rot 0 90 90>}]
 dyn_squiggleE_01 [{shapeType=Capsule, length=7.5154, radius=1.9624, center=<3 0 0>, rotation=<rot 0 90 90>}]
 dyn_squiggleE_01_Base [{shapeType=Capsule, length=0.5339, radius=0.5, rotation=<rot 0 -90 0>}]
 dyn_squiggleE_02 [{shapeType=Capsule, length=9.8374, radius=1.9624, center=<3 0 0>, rotation=<rot 0 90 90>}]
 dyn_squiggleE_03 [{shapeType=Capsule, length=9.0573, radius=1.9624, center=<3 0 0>, rotation=<rot 0 90 90>}]
 dyn_squiggleF_01 [{shapeType=Capsule, length=10.2303, radius=1.9624, center=<3 0 0>, rotation=<rot 0 90 90>}]
 dyn_squiggleF_01_Base [{shapeType=Capsule, length=0.5339, radius=0.5, rotation=<rot 0 -90 0>}]
 dyn_squiggleF_02 [{shapeType=Capsule, length=9.8374, radius=1.9624, center=<3 0 0>, rotation=<rot 0 90 90>}]
 dyn_squiggleF_03 [{shapeType=Capsule, length=3.8671, radius=1.9624, center=<2.0082 0 0>, rotation=<rot 0 90 90>}]
 dyn_squiggleG_01 [{shapeType=Capsule, length=9.5797, radius=1.9624, center=<3 0 0>, rotation=<rot 0 90 90>}]
 dyn_squiggleG_01_Base [{shapeType=Capsule, length=0.5339, radius=0.5, rotation=<rot 0 -90 0>}]
 dyn_squiggleG_02 [{shapeType=Capsule, length=9.8374, radius=1.9624, center=<3 0 0>, rotation=<rot 0 90 90>}]
 dyn_squiggleG_03 [{shapeType=Capsule, length=2.4338, radius=1.9624, center=<2.2186 0 0>, rotation=<rot 0 90 90>}]
 dyn_squiggleH_01 [{shapeType=Capsule, length=8.4315, radius=1.9624, center=<3 0 0>, rotation=<rot 0 90 90>}]
 dyn_squiggleH_01_Base [{shapeType=Capsule, length=0.5339, radius=0.5, rotation=<rot 0 -90 0>}]
 dyn_squiggleH_02 [{shapeType=Capsule, length=9.8374, radius=1.9624, center=<3 0 0>, rotation=<rot 0 90 90>}]
 dyn_squiggleH_03 [{shapeType=Capsule, length=9.0573, radius=1.9624, center=<3 0 0>, rotation=<rot 0 90 90>}]
 dyn_squiggleJ_01 [{shapeType=Capsule, length=6.2417, radius=1.9624, center=<3 0 0>, rotation=<rot 0 90 90>}]
 dyn_squiggleJ_01_Base [{shapeType=Capsule, length=0.5339, radius=0.5, rotation=<rot 0 -90 0>}]
 dyn_squiggleJ_02 [{shapeType=Capsule, length=4.9464, radius=1.9624, center=<3 0 0>, rotation=<rot 0 90 90>}]
 dyn_squiggleJ_03 [{shapeType=Capsule, length=5.6839, radius=1.9624, center=<1.4343 0 0.0001>, rotation=<rot 0 90 90>}]
 dyn_squiggleK_01 [{shapeType=Capsule, length=10.7146, radius=1.9624, center=<3 0 0>, rotation=<rot 0 90 90>}]
 dyn_squiggleK_01_Base [{shapeType=Capsule, length=0.5339, radius=0.5, rotation=<rot 0 -90 0>}]
 dyn_squiggleK_02 [{shapeType=Capsule, length=9.8374, radius=1.9624, center=<3 0 0>, rotation=<rot 0 90 90>}]
 dyn_squiggleK_03 [{shapeType=Capsule, length=6.6694, radius=1.9624, center=<2.0677 0 0.0001>, rotation=<rot 0 90 90>}]
 dyn_squiggleL_01 [{shapeType=Capsule, length=10.4639, radius=1.9624, center=<3 0 0>, rotation=<rot 0 90 90>}]
 dyn_squiggleL_01_Base [{shapeType=Capsule, length=0.5339, radius=0.5, rotation=<rot 0 -90 0>}]
 dyn_squiggleL_02 [{shapeType=Capsule, length=9.8374, radius=1.9624, center=<3 0 0>, rotation=<rot 0 90 90>}]
 dyn_squiggleL_03 [{shapeType=Capsule, length=5.7564, radius=1.9624, center=<1.4143 0 0>, rotation=<rot 0 90 90>}]
 dyn_squiggleM_01 [{shapeType=Capsule, length=7.8639, radius=1.9624, center=<3 0 0>, rotation=<rot 0 90 90>}]
 dyn_squiggleM_01_Base [{shapeType=Capsule, length=0.5339, radius=0.5, rotation=<rot 0 -90 0>}]
 dyn_squiggleM_02 [{shapeType=Capsule, length=9.8374, radius=1.9624, center=<3 0 0>, rotation=<rot 0 90 90>}]
 dyn_squiggleM_03 [{shapeType=Capsule, length=3.9793, radius=1.9624, center=<3 0 0>, rotation=<rot 0 90 90>}]
 dyn_squiggleN_01 [{shapeType=Capsule, length=12.4704, radius=1.9624, center=<3 0 0>, rotation=<rot 0 90 90>}]
 dyn_squiggleN_01_Base [{shapeType=Capsule, length=0.5339, radius=0.5, rotation=<rot 0 -90 0>}]
 dyn_squiggleN_02 [{shapeType=Capsule, length=9.8374, radius=1.9624, center=<3 0 0>, rotation=<rot 0 90 90>}]
 dyn_squiggleP_01 [{shapeType=Capsule, length=10.2315, radius=1.9624, center=<3 0 0>, rotation=<rot 0 90 90>}]
 dyn_squiggleP_01_Base [{shapeType=Capsule, length=0.5339, radius=0.5, rotation=<rot 0 -90 0>}]
 dyn_squiggleP_02 [{shapeType=Capsule, length=9.8374, radius=1.9624, center=<3 0 0>, rotation=<rot 0 90 90>}]
 dyn_squiggleP_03 [{shapeType=Capsule, length=9.0573, radius=1.9624, center=<3 0 0>, rotation=<rot 0 90 90>}]
 dyn_squiggleR_01 [{shapeType=Capsule, length=10.5494, radius=1.9624, center=<3 0 0>, rotation=<rot 0 90 90>}]
 dyn_squiggleR_01_Base [{shapeType=Capsule, length=0.5339, radius=0.5, rotation=<rot 0 -90 0>}]
 dyn_squiggleR_02 [{shapeType=Capsule, length=9.8374, radius=1.9624, center=<3 0 0>, rotation=<rot 0 90 90>}]
 dyn_squiggleR_03 [{shapeType=Capsule, length=9.0573, radius=1.9624, center=<3 0 0>, rotation=<rot 0 90 90>}]
 dyn_squiggleS_01 [{shapeType=Capsule, length=5.3803, radius=1.9624, center=<3 0 0>, rotation=<rot 0 90 90>}]
 dyn_squiggleS_01_Base [{shapeType=Capsule, length=0.5339, radius=0.5, rotation=<rot 0 -90 0>}]
 dyn_squiggleS_02 [{shapeType=Capsule, length=4.8319, radius=1.9624, center=<3 0 0>, rotation=<rot 0 90 90>}]
 dyn_squiggleS_03 [{shapeType=Capsule, length=3.15, radius=1.9624, center=<1.1537 0 0.0001>, rotation=<rot 0 90 90>}]
 dyn_squiggleT_01 [{shapeType=Capsule, length=9.4409, radius=1.9624, center=<3 0 0>, rotation=<rot 0 90 90>}]
 dyn_squiggleT_01_Base [{shapeType=Capsule, length=0.5339, radius=0.5, rotation=<rot 0 -90 0>}]
 dyn_squiggleT_02 [{shapeType=Capsule, length=9.8374, radius=1.9624, center=<3 0 0>, rotation=<rot 0 90 90>}]
 dyn_squiggleT_03 [{shapeType=Capsule, length=9.0573, radius=1.9624, center=<3 0 0>, rotation=<rot 0 90 90>}]
 head [{shapeType=Box, extentX=6.9525, extentY=10.7931, extentZ=7.8648, center=<1.6069 0.0828 0.6171>, rotation=<rot 0 -8.3538 -87.3402>}]
}
constraints [
 {
  bone1Name=dyn_headBase
  bone2Name=head
  swing1LimitDegrees=45
  swing1Motion=Free
  swing2LimitDegrees=45
  swing2Motion=Free
  twistLimitDegrees=45
  twistMotion=Free
 }
 {
  bone1Name=dyn_squiggleB_01_Base
  bone2Name=dyn_headBase
  swing1LimitDegrees=45
  swing1Motion=Free
  swing2LimitDegrees=45
  swing2Motion=Free
  twistLimitDegrees=45
  twistMotion=Free
 }
 {bone1Name=dyn_squiggleB_01, bone2Name=dyn_squiggleB_01_Base, swing1LimitDegrees=8, swing2LimitDegrees=8, twistLimitDegrees=8}
 {bone1Name=dyn_squiggleB_02, bone2Name=dyn_squiggleB_01, swing1LimitDegrees=8, swing2LimitDegrees=8, twistLimitDegrees=8}
 {bone1Name=dyn_squiggleB_03, bone2Name=dyn_squiggleB_02, swing1LimitDegrees=8, swing2LimitDegrees=8, twistLimitDegrees=8}
 {
  bone1Name=dyn_squiggleL_01_Base
  bone2Name=dyn_headBase
  swing1LimitDegrees=45
  swing1Motion=Free
  swing2LimitDegrees=45
  swing2Motion=Free
  twistLimitDegrees=45
  twistMotion=Free
 }
 {bone1Name=dyn_squiggleL_01, bone2Name=dyn_squiggleL_01_Base, swing1LimitDegrees=8, swing2LimitDegrees=8, twistLimitDegrees=8}
 {bone1Name=dyn_squiggleL_02, bone2Name=dyn_squiggleL_01, swing1LimitDegrees=8, swing2LimitDegrees=8, twistLimitDegrees=8}
 {bone1Name=dyn_squiggleL_03, bone2Name=dyn_squiggleL_02, swing1LimitDegrees=8, swing2LimitDegrees=8, twistLimitDegrees=8}
 {
  bone1Name=dyn_squiggleE_01_Base
  bone2Name=dyn_headBase
  swing1LimitDegrees=45
  swing1Motion=Free
  swing2LimitDegrees=45
  swing2Motion=Free
  twistLimitDegrees=45
  twistMotion=Free
 }
 {bone1Name=dyn_squiggleE_01, bone2Name=dyn_squiggleE_01_Base, swing1LimitDegrees=8, swing2LimitDegrees=8, twistLimitDegrees=8}
 {bone1Name=dyn_squiggleE_02, bone2Name=dyn_squiggleE_01, swing1LimitDegrees=8, swing2LimitDegrees=8, twistLimitDegrees=8}
 {bone1Name=dyn_squiggleE_03, bone2Name=dyn_squiggleE_02, swing1LimitDegrees=8, swing2LimitDegrees=8, twistLimitDegrees=8}
 {
  bone1Name=dyn_squiggleN_01_Base
  bone2Name=dyn_headBase
  swing1LimitDegrees=45
  swing1Motion=Free
  swing2LimitDegrees=45
  swing2Motion=Free
  twistLimitDegrees=45
  twistMotion=Free
 }
 {bone1Name=dyn_squiggleN_01, bone2Name=dyn_squiggleN_01_Base, swing1LimitDegrees=8, swing2LimitDegrees=8, twistLimitDegrees=8}
 {bone1Name=dyn_squiggleN_02, bone2Name=dyn_squiggleN_01, swing1LimitDegrees=8, swing2LimitDegrees=8, twistLimitDegrees=8}
 {
  bone1Name=dyn_squiggleF_01_Base
  bone2Name=dyn_headBase
  swing1LimitDegrees=45
  swing1Motion=Free
  swing2LimitDegrees=45
  swing2Motion=Free
  twistLimitDegrees=45
  twistMotion=Free
 }
 {bone1Name=dyn_squiggleF_01, bone2Name=dyn_squiggleF_01_Base, swing1LimitDegrees=8, swing2LimitDegrees=8, twistLimitDegrees=8}
 {bone1Name=dyn_squiggleF_02, bone2Name=dyn_squiggleF_01, swing1LimitDegrees=8, swing2LimitDegrees=8, twistLimitDegrees=8}
 {bone1Name=dyn_squiggleF_03, bone2Name=dyn_squiggleF_02, swing1LimitDegrees=8, swing2LimitDegrees=8, twistLimitDegrees=8}
 {
  bone1Name=dyn_squiggleD_01_Base
  bone2Name=dyn_headBase
  swing1LimitDegrees=45
  swing1Motion=Free
  swing2LimitDegrees=45
  swing2Motion=Free
  twistLimitDegrees=45
  twistMotion=Free
 }
 {bone1Name=dyn_squiggleD_01, bone2Name=dyn_squiggleD_01_Base, swing1LimitDegrees=8, swing2LimitDegrees=8, twistLimitDegrees=8}
 {bone1Name=dyn_squiggleD_02, bone2Name=dyn_squiggleD_01, swing1LimitDegrees=8, swing2LimitDegrees=8, twistLimitDegrees=8}
 {bone1Name=dyn_squiggleD_03, bone2Name=dyn_squiggleD_02, swing1LimitDegrees=8, swing2LimitDegrees=8, twistLimitDegrees=8}
 {
  bone1Name=dyn_squiggleP_01_Base
  bone2Name=dyn_headBase
  swing1LimitDegrees=45
  swing1Motion=Free
  swing2LimitDegrees=45
  swing2Motion=Free
  twistLimitDegrees=45
  twistMotion=Free
 }
 {bone1Name=dyn_squiggleP_01, bone2Name=dyn_squiggleP_01_Base, swing1LimitDegrees=8, swing2LimitDegrees=8, twistLimitDegrees=8}
 {bone1Name=dyn_squiggleP_02, bone2Name=dyn_squiggleP_01, swing1LimitDegrees=8, swing2LimitDegrees=8, twistLimitDegrees=8}
 {bone1Name=dyn_squiggleP_03, bone2Name=dyn_squiggleP_02, swing1LimitDegrees=8, swing2LimitDegrees=8, twistLimitDegrees=8}
 {
  bone1Name=dyn_squiggleR_01_Base
  bone2Name=dyn_headBase
  swing1LimitDegrees=45
  swing1Motion=Free
  swing2LimitDegrees=45
  swing2Motion=Free
  twistLimitDegrees=45
  twistMotion=Free
 }
 {bone1Name=dyn_squiggleR_01, bone2Name=dyn_squiggleR_01_Base, swing1LimitDegrees=8, swing2LimitDegrees=8, twistLimitDegrees=8}
 {bone1Name=dyn_squiggleR_02, bone2Name=dyn_squiggleR_01, swing1LimitDegrees=8, swing2LimitDegrees=8, twistLimitDegrees=8}
 {bone1Name=dyn_squiggleR_03, bone2Name=dyn_squiggleR_02, swing1LimitDegrees=8, swing2LimitDegrees=8, twistLimitDegrees=8}
 {
  bone1Name=dyn_squiggleT_01_Base
  bone2Name=dyn_headBase
  swing1LimitDegrees=45
  swing1Motion=Free
  swing2LimitDegrees=45
  swing2Motion=Free
  twistLimitDegrees=45
  twistMotion=Free
 }
 {bone1Name=dyn_squiggleT_01, bone2Name=dyn_squiggleT_01_Base, swing1LimitDegrees=8, swing2LimitDegrees=8, twistLimitDegrees=8}
 {bone1Name=dyn_squiggleT_02, bone2Name=dyn_squiggleT_01, swing1LimitDegrees=8, swing2LimitDegrees=8, twistLimitDegrees=8}
 {bone1Name=dyn_squiggleT_03, bone2Name=dyn_squiggleT_02, swing1LimitDegrees=8, swing2LimitDegrees=8, twistLimitDegrees=8}
 {
  bone1Name=dyn_squiggleK_01_Base
  bone2Name=dyn_headBase
  swing1LimitDegrees=45
  swing1Motion=Free
  swing2LimitDegrees=45
  swing2Motion=Free
  twistLimitDegrees=45
  twistMotion=Free
 }
 {bone1Name=dyn_squiggleK_01, bone2Name=dyn_squiggleK_01_Base, swing1LimitDegrees=8, swing2LimitDegrees=8, twistLimitDegrees=8}
 {bone1Name=dyn_squiggleK_02, bone2Name=dyn_squiggleK_01, swing1LimitDegrees=8, swing2LimitDegrees=8, twistLimitDegrees=8}
 {bone1Name=dyn_squiggleK_03, bone2Name=dyn_squiggleK_02, swing1LimitDegrees=8, swing2LimitDegrees=8, twistLimitDegrees=8}
 {
  bone1Name=dyn_squiggleH_01_Base
  bone2Name=dyn_headBase
  swing1LimitDegrees=45
  swing1Motion=Free
  swing2LimitDegrees=45
  swing2Motion=Free
  twistLimitDegrees=45
  twistMotion=Free
 }
 {bone1Name=dyn_squiggleH_01, bone2Name=dyn_squiggleH_01_Base, swing1LimitDegrees=8, swing2LimitDegrees=8, twistLimitDegrees=8}
 {bone1Name=dyn_squiggleH_02, bone2Name=dyn_squiggleH_01, swing1LimitDegrees=8, swing2LimitDegrees=8, twistLimitDegrees=8}
 {bone1Name=dyn_squiggleH_03, bone2Name=dyn_squiggleH_02, swing1LimitDegrees=8, swing2LimitDegrees=8, twistLimitDegrees=8}
 {
  bone1Name=dyn_squiggleG_01_Base
  bone2Name=dyn_headBase
  swing1LimitDegrees=45
  swing1Motion=Free
  swing2LimitDegrees=45
  swing2Motion=Free
  twistLimitDegrees=45
  twistMotion=Free
 }
 {bone1Name=dyn_squiggleG_01, bone2Name=dyn_squiggleG_01_Base, swing1LimitDegrees=8, swing2LimitDegrees=8, twistLimitDegrees=8}
 {bone1Name=dyn_squiggleG_02, bone2Name=dyn_squiggleG_01, swing1LimitDegrees=8, swing2LimitDegrees=8, twistLimitDegrees=8}
 {bone1Name=dyn_squiggleG_03, bone2Name=dyn_squiggleG_02, swing1LimitDegrees=8, swing2LimitDegrees=8, twistLimitDegrees=8}
 {
  bone1Name=dyn_squiggleM_01_Base
  bone2Name=dyn_headBase
  swing1LimitDegrees=45
  swing1Motion=Free
  swing2LimitDegrees=45
  swing2Motion=Free
  twistLimitDegrees=45
  twistMotion=Free
 }
 {bone1Name=dyn_squiggleM_01, bone2Name=dyn_squiggleM_01_Base, swing1LimitDegrees=8, swing2LimitDegrees=8, twistLimitDegrees=8}
 {bone1Name=dyn_squiggleM_02, bone2Name=dyn_squiggleM_01, swing1LimitDegrees=8, swing2LimitDegrees=8, twistLimitDegrees=8}
 {bone1Name=dyn_squiggleM_03, bone2Name=dyn_squiggleM_02, swing1LimitDegrees=8, swing2LimitDegrees=8, twistLimitDegrees=8}
 {
  bone1Name=dyn_squiggleC_01_Base
  bone2Name=dyn_headBase
  swing1LimitDegrees=45
  swing1Motion=Free
  swing2LimitDegrees=45
  swing2Motion=Free
  twistLimitDegrees=45
  twistMotion=Free
 }
 {bone1Name=dyn_squiggleC_01, bone2Name=dyn_squiggleC_01_Base, swing1LimitDegrees=20, swing2LimitDegrees=20, twistLimitDegrees=8}
 {bone1Name=dyn_squiggleC_02, bone2Name=dyn_squiggleC_01, swing1LimitDegrees=30, swing2LimitDegrees=20, twistLimitDegrees=8}
 {
  bone1Name=dyn_squiggleJ_01_Base
  bone2Name=dyn_headBase
  swing1LimitDegrees=45
  swing1Motion=Free
  swing2LimitDegrees=45
  swing2Motion=Free
  twistLimitDegrees=45
  twistMotion=Free
 }
 {bone1Name=dyn_squiggleJ_01, bone2Name=dyn_squiggleJ_01_Base, swing1LimitDegrees=8, swing2LimitDegrees=8, twistLimitDegrees=8}
 {bone1Name=dyn_squiggleJ_02, bone2Name=dyn_squiggleJ_01, swing1LimitDegrees=8, swing2LimitDegrees=8, twistLimitDegrees=8}
 {bone1Name=dyn_squiggleJ_03, bone2Name=dyn_squiggleJ_02, swing1LimitDegrees=8, swing2LimitDegrees=8, twistLimitDegrees=8}
 {
  bone1Name=dyn_squiggleS_01_Base
  bone2Name=dyn_headBase
  swing1LimitDegrees=45
  swing1Motion=Free
  swing2LimitDegrees=45
  swing2Motion=Free
  twistLimitDegrees=45
  twistMotion=Free
 }
 {bone1Name=dyn_squiggleS_01, bone2Name=dyn_squiggleS_01_Base, swing1LimitDegrees=8, swing2LimitDegrees=8, twistLimitDegrees=8}
 {bone1Name=dyn_squiggleS_02, bone2Name=dyn_squiggleS_01, swing1LimitDegrees=8, swing2LimitDegrees=8, twistLimitDegrees=8}
 {bone1Name=dyn_squiggleS_03, bone2Name=dyn_squiggleS_02, swing1LimitDegrees=8, swing2LimitDegrees=8, twistLimitDegrees=8}
 {
  bone1Name=dyn_squiggleA_01_Base
  bone2Name=dyn_headBase
  swing1LimitDegrees=45
  swing1Motion=Free
  swing2LimitDegrees=45
  swing2Motion=Free
  twistLimitDegrees=45
  twistMotion=Free
 }
 {bone1Name=dyn_squiggleA_01, bone2Name=dyn_squiggleA_01_Base, swing1LimitDegrees=8, swing2LimitDegrees=8, twistLimitDegrees=8}
 {bone1Name=dyn_squiggleA_02, bone2Name=dyn_squiggleA_01, swing1LimitDegrees=8, swing2LimitDegrees=8, twistLimitDegrees=8}
 {bone1Name=dyn_squiggleA_03, bone2Name=dyn_squiggleA_02, swing1LimitDegrees=8, swing2LimitDegrees=8, twistLimitDegrees=8}
]
